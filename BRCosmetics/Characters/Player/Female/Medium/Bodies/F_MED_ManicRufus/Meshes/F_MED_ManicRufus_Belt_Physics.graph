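# Physics asset for the belt mesh of the ManicRufus medium female player cosmetic, defining bodies for two dangling belt strands.
# Each strand is a chain of three identical dynamic-belt capsules hung from a tiny base capsule, whose constraint to the pelvis leaves swing and twist free.
# Chain joints have narrow, asymmetric swing limits and zero twist limit.
# Larger pelvis and right-thigh capsules approximate the body, with no left-thigh body.
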PhysicsAsset /BRCosmetics/Characters/Player/Female/Medium/Bodies/F_MED_ManicRufus/Meshes/F_MED_ManicRufus_Belt_Physics
bodies {
 base_belt_01_01 [{shapeType=Capsule, length=0.505, radius=0.505, shapeName=base_belt_01_01_capsule, rotation=<rot 0 -90 0>}]
 base_belt_02_01 [{shapeType=Capsule, length=0.505, radius=0.505, shapeName=base_belt_02_01_capsule, rotation=<rot 0 -90 0>}]
 dyn_belt_01_01 [{shapeType=Capsule, length=3.5099, radius=1.4901, shapeName=dyn_belt_01_01_capsule, center=<2 0 0>, rotation=<rot 90 0 0>}]
 dyn_belt_01_02 [{shapeType=Capsule, length=3.5099, radius=1.4901, shapeName=dyn_belt_01_02_capsule, center=<2 0 0>, rotation=<rot 90 0 0>}]
 dyn_belt_01_03 [{shapeType=Capsule, length=3.5099, radius=1.4901, shapeName=dyn_belt_01_03_capsule, center=<2 0 0>, rotation=<rot 90 0 0>}]
 dyn_belt_02_01 [{shapeType=Capsule, length=3.5099, radius=1.4901, shapeName=dyn_belt_02_01_capsule, center=<2 0 0>, rotation=<rot 90 0 0>}]
 dyn_belt_02_02 [{shapeType=Capsule, length=3.5099, radius=1.4901, shapeName=dyn_belt_02_02_capsule, center=<2 0 0>, rotation=<rot 90 0 0>}]
 dyn_belt_02_03 [{shapeType=Capsule, length=3.5099, radius=1.4901, shapeName=dyn_belt_02_03_capsule, center=<2 0 0>, rotation=<rot 90 0 0>}]
 pelvis [{shapeType=Capsule, length=3.3852, radius=13.2851, rotation=<rot 0 -90 0>}]
 thigh_r [{shapeType=Capsule, length=26.5756, radius=8.1253, center=<27.8448 1.1114 -0.3445>, rotation=<rot 0.322 -81.4727 -89.3206>}]
}
constraints [
 {bone1Name=thigh_r, bone2Name=pelvis, swing1LimitDegrees=45, swing2LimitDegrees=45, twistLimitDegrees=45}
 {
  bone1Name=base_belt_02_01
  bone2Name=pelvis
  swing1LimitDegrees=45
  swing1Motion=Free
  swing2LimitDegrees=45
  swing2Motion=Free
  twistLimitDegrees=45
  twistMotion=Free
 }
 {bone1Name=dyn_belt_02_01, bone2Name=base_belt_02_01, swing1LimitDegrees=10, swing2LimitDegrees=20, twistLimitDegrees=0}
 {bone1Name=dyn_belt_02_02, bone2Name=dyn_belt_02_01, swing1LimitDegrees=10, swing2LimitDegrees=20, twistLimitDegrees=0}
 {bone1Name=dyn_belt_02_03, bone2Name=dyn_belt_02_02, swing1LimitDegrees=10, swing2LimitDegrees=20, twistLimitDegrees=0}
 {
  bone1Name=base_belt_01_01
  bone2Name=pelvis
  swing1LimitDegrees=45
  swing1Motion=Free
  swing2LimitDegrees=45
  swing2Motion=Free
  twistLimitDegrees=45
  twistMotion=Free
 }
 {bone1Name=dyn_belt_01_01, bone2Name=base_belt_01_01, swing1LimitDegrees=10, swing2LimitDegrees=20, twistLimitDegrees=0}
 {bone1Name=dyn_belt_01_02, bone2Name=dyn_belt_01_01, swing1LimitDegrees=10, swing2LimitDegrees=20, twistLimitDegrees=0}
 {bone1Name=dyn_belt_01_03, bone2Name=dyn_belt_01_02, swing1LimitDegrees=10, swing2LimitDegrees=20, twistLimitDegrees=0}
]
