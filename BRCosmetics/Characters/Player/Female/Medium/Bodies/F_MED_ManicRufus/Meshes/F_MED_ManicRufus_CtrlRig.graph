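# Control Rig for the F_MED_ManicRufus cosmetic body, procedurally posing corrective and accessory bones.
# After the shared DeformRig function, spherical pose readers on the thighs swing skirt and loincloth bones.
# Bags and grenades follow skirt bones through weighted parent and aim constraints, with DeformRig thigh attributes adding offsets.
# A crouch pass blends poses by pelvis height and thigh weights, then upperarm and spine readers adjust shoulderpad and skirt.
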ControlRigBlueprint /BRCosmetics/Characters/Player/Female/Medium/Bodies/F_MED_ManicRufus/Meshes/F_MED_ManicRufus_CtrlRig
graphs [
 New Function_ContainedGraph {nodes [
   n1 = RigVMFunctionEntryNode()
   RigVMFunction_Sequence = &RigVMFunction_Sequence(ExecuteContext=@n1.ExecuteContext)
   RigVMFunction_Sequence_1 = RigVMAggregateNode(ExecuteContext=@RigVMFunction_Sequence.A)
   RigVMFunction_Sequence_2 = RigVMAggregateNode(ExecuteContext=@RigVMFunction_Sequence_1.A)
   n2 = &RigUnit_SphericalPoseReader(
    ActiveRegionScaleFactors=(PositiveWidth=1,NegativeWidth=1,PositiveHeight=1,NegativeHeight=1)
    ActiveRegionSize=0.1
    Debug=(bDrawDebug=false,bDraw2D=False,bDrawLocalAxes=False,DebugScale=25,DebugSegments=20,DebugThickness=0.25)
    DriverAxis=(X=-1,Y=0,Z=0)
    DriverItem=(Type=Bone,Name="thigh_l")
    FalloffRegionScaleFactors=(PositiveWidth=0.5,NegativeWidth=1,PositiveHeight=1,NegativeHeight=1)
    FalloffSize=0.65
    FlipHeightScaling=False
    FlipWidthScaling=False
    OptionalParentItem=(Type=Bone,Name="None")
    RotationOffset=(X=90,Y=0,Z=-30)
    ExecutePin=@RigVMFunction_Sequence_2.A
   )
   > &RigUnit_ModifyTransforms(
    ItemToModify="((Item=(Type=Bone,Name=\"auto_skirt_front_lf\"),Transform=(Rotation=(X=0,Y=0,Z=-0.1736,W=0.9848),Translation=(X=0,Y=0,Z=0),Scale3D=(X=1,Y=1,Z=1))),(Item=(Type=Bone,Name=\"auto_skirt_front_side_lf\"),Transform=(Rotation=(X=0,Y=0,Z=0,W=1),Translation=(X=0,Y=0,Z= ...(truncated)"
    WeightMinimum=0
    Weight=@n2.OutputParam
   )
   > &RigUnit_ModifyTransforms(
    ItemToModify="((Item=(Type=Bone,Name=\"offset_skirt_front_lf\"),Transform=(Rotation=(X=0,Y=-0.766,Z=0,W=0.6428),Translation=(X=0,Y=0,Z=0),Scale3D=(X=1,Y=1,Z=1))),(Item=(Type=Bone,Name=\"offset_skirt_front_side_lf\"),Transform=(Rotation=(X=0,Y=-0.5373,Z=0,W=0.8434),Translati ...(truncated)"
    WeightMinimum=0
    Weight=@n2.OutputParam
   )
   > &RigUnit_ModifyTransforms(
    ItemToModify="((Item=(Type=Bone,Name=\"auto_skirt_front_side_lf\"),Transform=(Rotation=(X=0,Y=0,Z=-0.4226,W=0.9063),Translation=(X=0,Y=0,Z=0),Scale3D=(X=1,Y=1,Z=1))),(Item=(Type=Bone,Name=\"auto_skirt_front_lf\"),Transform=(Rotation=(X=0,Y=0,Z=-0.5,W=0.866),Translation=(X=0 ...(truncated)"
    WeightMinimum=0
    Weight=@n2.OutputParam
   )
   n3 = &RigUnit_SphericalPoseReader(
    ActiveRegionScaleFactors=(PositiveWidth=1,NegativeWidth=1,PositiveHeight=1,NegativeHeight=1)
    ActiveRegionSize=0.1
    Debug=(bDrawDebug=false,bDraw2D=False,bDrawLocalAxes=False,DebugScale=25,DebugSegments=20,DebugThickness=0.25)
    DriverAxis=(X=-1,Y=0,Z=0)
    DriverItem=(Type=Bone,Name="thigh_l")
    FalloffRegionScaleFactors=(PositiveWidth=0.5,NegativeWidth=1,PositiveHeight=0.7,NegativeHeight=0.7)
    FalloffSize=0.65
    FlipHeightScaling=False
    FlipWidthScaling=False
    OptionalParentItem=(Type=Bone,Name="None")
    RotationOffset=(X=180,Y=30,Z=0)
    ExecutePin=@RigVMFunction_Sequence_2.B
   )
   > &RigUnit_ModifyTransforms(
    ItemToModify="((Item=(Type=Bone,Name=\"offset_skirt_side_lf\"),Transform=(Rotation=(X=0,Y=-0.8192,Z=0,W=0.5736),Translation=(X=0,Y=0,Z=0),Scale3D=(X=1,Y=1,Z=1))),(Item=(Type=Bone,Name=\"offset_skirt_front_side_lf\"),Transform=(Rotation=(X=0,Y=-0.7071,Z=0,W=0.7071),Translati ...(truncated)"
    WeightMinimum=0
    Weight=@n3.OutputParam
   )
   > n4 = &RigUnit_SphericalPoseReader(
    ActiveRegionScaleFactors=(PositiveWidth=1,NegativeWidth=1,PositiveHeight=1,NegativeHeight=1)
    ActiveRegionSize=0.1
    Debug=(bDrawDebug=false,bDraw2D=False,bDrawLocalAxes=False,DebugScale=25,DebugSegments=20,DebugThickness=0.25)
    DriverAxis=(X=-1,Y=0,Z=0)
    DriverItem=(Type=Bone,Name="thigh_l")
    FalloffRegionScaleFactors=(PositiveWidth=0.5,NegativeWidth=1,PositiveHeight=0.8,NegativeHeight=0.7)
    FalloffSize=0.65
    FlipHeightScaling=False
    FlipWidthScaling=False
    OptionalParentItem=(Type=Bone,Name="None")
    RotationOffset=(X=-90,Y=0,Z=30)
   )
   > &RigUnit_ModifyTransforms(
    ItemToModify="((Item=(Type=Bone,Name=\"offset_skirt_back_lf\"),Transform=(Rotation=(X=0,Y=-0.8192,Z=0,W=0.5736),Translation=(X=0,Y=0,Z=0),Scale3D=(X=1,Y=1,Z=1))),(Item=(Type=Bone,Name=\"offset_skirt_back_side_lf\"),Transform=(Rotation=(X=0,Y=-0.5,Z=0,W=0.866),Translation=(X ...(truncated)"
    WeightMinimum=0
    Weight=@n4.OutputParam
   )
   > &RigUnit_ModifyTransforms(
    ItemToModify="((Item=(Type=Bone,Name=\"offset_skirt_front_side_lf\"),Transform=(Rotation=(X=0,Y=-0.1736,Z=0,W=0.9848),Translation=(X=0,Y=0,Z=0),Scale3D=(X=1,Y=1,Z=1))),(Item=(Type=Bone,Name=\"offset_skirt_side_lf\"),Transform=(Rotation=(X=0,Y=-0.1736,Z=0,W=0.9848),Translati ...(truncated)"
    WeightMinimum=0
    Weight=@n4.OutputParam
   )
   RigVMFunction_Sequence_3 = RigVMAggregateNode(ExecuteContext=@RigVMFunction_Sequence_1.B)
   n5 = &RigUnit_SphericalPoseReader(
    ActiveRegionScaleFactors=(PositiveWidth=1,NegativeWidth=1,PositiveHeight=1,NegativeHeight=1)
    ActiveRegionSize=0.1
    Debug=(bDrawDebug=false,bDraw2D=False,bDrawLocalAxes=False,DebugScale=25,DebugSegments=20,DebugThickness=0.25)
    DriverAxis=(X=1,Y=0,Z=0)
    DriverItem=(Type=Bone,Name="thigh_r")
    FalloffRegionScaleFactors=(PositiveWidth=0.5,NegativeWidth=1,PositiveHeight=1,NegativeHeight=1)
    FalloffSize=0.65
    FlipHeightScaling=False
    FlipWidthScaling=False
    OptionalParentItem=(Type=Bone,Name="None")
    RotationOffset=(X=-90,Y=180,Z=-30)
    ExecutePin=@RigVMFunction_Sequence_3.A
   )
   > &RigUnit_ModifyTransforms(
    ItemToModify="((Item=(Type=Bone,Name=\"offset_skirt_front_rt\"),Transform=(Rotation=(X=0,Y=-0.8192,Z=0,W=0.5736),Translation=(X=0,Y=0,Z=0),Scale3D=(X=1,Y=1,Z=1))),(Item=(Type=Bone,Name=\"offset_skirt_front_side_rt\"),Transform=(Rotation=(X=0,Y=-0.5736,Z=0,W=0.8192),Translat ...(truncated)"
    WeightMinimum=0
    Weight=@n5.OutputParam
   )
   > &RigUnit_ModifyTransforms(
    ItemToModify="((Item=(Type=Bone,Name=\"auto_skirt_front_rt\"),Transform=(Rotation=(X=0,Y=0,Z=0.5,W=0.866),Translation=(X=0,Y=0,Z=0),Scale3D=(X=1,Y=1,Z=1))),(Item=(Type=Bone,Name=\"auto_skirt_front_side_rt\"),Transform=(Rotation=(X=0,Y=0,Z=0.4226,W=0.9063),Translation=(X=0,Y ...(truncated)"
    WeightMinimum=0
    Weight=@n5.OutputParam
   )
   n6 = &RigUnit_SphericalPoseReader(
    ActiveRegionScaleFactors=(PositiveWidth=1,NegativeWidth=1,PositiveHeight=1,NegativeHeight=1)
    ActiveRegionSize=0.1
    Debug=(bDrawDebug=false,bDraw2D=False,bDrawLocalAxes=False,DebugScale=25,DebugSegments=20,DebugThickness=0.25)
    DriverAxis=(X=1,Y=0,Z=0)
    DriverItem=(Type=Bone,Name="thigh_r")
    FalloffRegionScaleFactors=(PositiveWidth=0.5,NegativeWidth=1,PositiveHeight=0.7,NegativeHeight=0.7)
    FalloffSize=0.65
    FlipHeightScaling=false
    FlipWidthScaling=false
    OptionalParentItem=(Type=Bone,Name="None")
    RotationOffset=(X=0,Y=-30,Z=180)
    ExecutePin=@RigVMFunction_Sequence_3.B
   )
   > &RigUnit_ModifyTransforms(
    ItemToModify="((Item=(Type=Bone,Name=\"offset_skirt_side_rt\"),Transform=(Rotation=(X=0,Y=-0.8192,Z=0,W=0.5736),Translation=(X=0,Y=0,Z=0),Scale3D=(X=1,Y=1,Z=1))),(Item=(Type=Bone,Name=\"offset_skirt_front_side_rt\"),Transform=(Rotation=(X=0,Y=-0.5736,Z=0,W=0.8192),Translati ...(truncated)"
    WeightMinimum=0
    Weight=@n6.OutputParam
   )
   > &RigUnit_ModifyTransforms(
    ItemToModify="((Item=(Type=Bone,Name=\"offset_skirt_front_side_rt\"),Transform=(Rotation=(X=0,Y=-0.1736,Z=0,W=0.9848),Translation=(X=0,Y=0,Z=0),Scale3D=(X=1,Y=1,Z=1))),(Item=(Type=Bone,Name=\"offset_skirt_front_rt\"),Transform=(Rotation=(X=0,Y=-0.1736,Z=0,W=0.9848),Translat ...(truncated)"
    WeightMinimum=0
    Weight=@n6.OutputParam
   )
   n7 = &RigUnit_SphericalPoseReader(
    ActiveRegionScaleFactors=(PositiveWidth=1,NegativeWidth=1,PositiveHeight=1,NegativeHeight=1)
    ActiveRegionSize=0.1
    Debug=(bDrawDebug=false,bDraw2D=False,bDrawLocalAxes=False,DebugScale=25,DebugSegments=20,DebugThickness=0.25)
    DriverAxis=(X=1,Y=0,Z=0)
    DriverItem=(Type=Bone,Name="thigh_r")
    FalloffRegionScaleFactors=(PositiveWidth=0.5,NegativeWidth=1,PositiveHeight=0.7,NegativeHeight=0.7)
    FalloffSize=0.65
    FlipHeightScaling=true
    FlipWidthScaling=true
    OptionalParentItem=(Type=Bone,Name="None")
    RotationOffset=(X=90,Y=0,Z=30)
    ExecutePin=@RigVMFunction_Sequence_3.C
   )
   > &RigUnit_ModifyTransforms(
    ItemToModify="((Item=(Type=Bone,Name=\"offset_skirt_back_rt\"),Transform=(Rotation=(X=0,Y=-0.866,Z=0,W=0.5),Translation=(X=0,Y=0,Z=0),Scale3D=(X=1,Y=1,Z=1))),(Item=(Type=Bone,Name=\"offset_skirt_back_side_rt\"),Transform=(Rotation=(X=0,Y=-0.5,Z=0,W=0.866),Translation=(X=0,Y ...(truncated)"
    WeightMinimum=0
    Weight=@n7.OutputParam
   )
   > &RigUnit_ParentConstraint(
    AdvancedSettings=(InterpolationType=Average,RotationOrderForFilter=XZY)
    Child=(Type=Bone,Name="auto_loin_cloth_back_rt")
    Parents=((Item=(Type=Bone,Name="auto_loin_cloth_back_rt"),Weight=1))
    Weight=1
   )
   GreaterEqual = /Script/RigVM.RigVMFunction_MathDoubleGreaterEqual(A=@n2.OutputParam, B=@n5.OutputParam) {is=struct RigVMUnitNode, template="GreaterEqual::Execute(in A,in B,out Result)"}
   > &RigUnit_ModifyTransforms(
    ItemToModify=((Item=(Type=Bone,Name="offset_loin_cloth_front_cr"),Transform=(Rotation=(X=0,Y=-0.7071,Z=0,W=0.7071),Translation=(X=0,Y=0,Z=0),Scale3D=(X=1,Y=1,Z=1))))
    WeightMinimum=0
    Weight=RigVMDispatchNode(Condition=@GreaterEqual.Result, True=@n2.OutputParam, False=@n5.OutputParam) {template="DISPATCH_RigVMDispatch_If(in Condition,in True,in False,out Result)"}.Result
   )
   &RigUnit_ParentConstraint(
    AdvancedSettings=(InterpolationType=Average,RotationOrderForFilter=XZY)
    Child=(Type=Bone,Name="dyn_loin_cloth_back_02_lf")
    Parents=((Item=(Type=Bone,Name="dyn_skirt_back_02_lf"),Weight=1))
    Weight=1
    ExecutePin=@RigVMFunction_Sequence_1.C
   )
   > &RigUnit_ParentConstraint(
    AdvancedSettings=(InterpolationType=Average,RotationOrderForFilter=XZY)
    Child=(Type=Bone,Name="dyn_loin_cloth_back_02_rt")
    Parents=((Item=(Type=Bone,Name="dyn_skirt_back_02_rt"),Weight=1))
    Weight=1
   )
   RigVMFunctionReturnNode(ExecuteContext=@RigVMFunction_Sequence.B, left thigh=@n2.OutputParam, right thigh=@n5.OutputParam)
  ]}
 New Function_ContainedGraph {nodes [
   n1 = RigVMFunctionEntryNode()
   RigVMFunction_ControlFlowBranch = &RigVMFunction_ControlFlowBranch(Condition=@n1.Debug, ExecuteContext=@n1.ExecuteContext)
   RigVMFunctionReturnNode(ExecuteContext=@RigVMFunction_ControlFlowBranch.Completed)
   RigVMFunction_Sequence = &RigVMFunction_Sequence(ExecuteContext=@RigVMFunction_ControlFlowBranch.True)
   RigUnit_GetTransform = &RigUnit_GetTransform(Item=(Type=Control,Name="thigh_l_ctrl"), bInitial=False)
   RigUnit_SetTransform = &RigUnit_SetTransform(Item=(Type=Bone,Name="thigh_l"), Weight=1, bInitial=False, Value=@RigUnit_GetTransform.Transform, ExecutePin=@RigVMFunction_Sequence.A)
   > &RigUnit_SetTransform(
    Item=(Type=Bone,Name="thigh_r")
    Weight=1
    bInitial=False
    Value=&RigUnit_GetTransform(Item=(Type=Control,Name="thigh_r_ctrl"), bInitial=False).Transform
   )
   > &RigUnit_SetTransform(
    Item=(Type=Bone,Name="calf_r")
    Weight=1
    bInitial=False
    Value=&RigUnit_GetTransform(Item=(Type=Control,Name="calf_r_ctrl"), bInitial=False).Transform
   )
   > &RigUnit_SetTransform(
    Item=(Type=Bone,Name="calf_l")
    Weight=1
    bInitial=False
    Value=&RigUnit_GetTransform(Item=(Type=Control,Name="calf_l_ctrl"), bInitial=False).Transform
   )
  ]}
 New Function_ContainedGraph {nodes [
   n1 = RigVMFunctionEntryNode()
   RigVMFunction_Sequence = &RigVMFunction_Sequence(ExecuteContext=@n1.ExecuteContext)
   RigVMFunction_Sequence_1 = RigVMAggregateNode(ExecuteContext=@RigVMFunction_Sequence.A)
   &RigUnit_ParentConstraint(
    AdvancedSettings=(RotationOrderForFilter=XZY)
    Child=(Type=Bone,Name="aim_bags_lf")
    Parents="((Item=(Type=Bone,Name=\"dyn_skirt_side_03_lf\"),Weight=0.0054),(Item=(Type=Bone,Name=\"dyn_skirt_front_01_lf\"),Weight=0.0059),(Item=(Type=Bone,Name=\"dyn_skirt_front_side_01_lf\"),Weight=0.0164),(Item=(Type=Bone,Name=\"dyn_skirt_side_02_lf\"),Weight=0.0212),(Ite ...(truncated)"
    Weight=1
    ExecutePin=@RigVMFunction_Sequence_1.A
   )
   > &RigUnit_ParentConstraint(
    AdvancedSettings=(RotationOrderForFilter=XZY)
    Child=(Type=Bone,Name="aim_granade_01_lf")
    Parents="((Item=(Type=Bone,Name=\"dyn_skirt_side_01_lf\"),Weight=0.0066),(Item=(Type=Bone,Name=\"dyn_skirt_front_side_01_lf\"),Weight=0.0071),(Item=(Type=Bone,Name=\"dyn_skirt_front_side_03_lf\"),Weight=0.0237),(Item=(Type=Bone,Name=\"dyn_skirt_back_side_03_lf\"),Weight=0. ...(truncated)"
    Weight=1
   )
   > &RigUnit_ParentConstraint(
    AdvancedSettings=(RotationOrderForFilter=XZY)
    Child=(Type=Bone,Name="aim_granade_02_lf")
    Parents="((Item=(Type=Bone,Name=\"dyn_skirt_side_03_lf\"),Weight=0.0143),(Item=(Type=Bone,Name=\"dyn_skirt_front_side_02_lf\"),Weight=0.0278),(Item=(Type=Bone,Name=\"dyn_skirt_side_02_lf\"),Weight=0.0278),(Item=(Type=Bone,Name=\"dyn_skirt_back_01_lf\"),Weight=0.0293),(Item ...(truncated)"
    Weight=1
   )
   > &RigUnit_ParentConstraint(
    AdvancedSettings=(RotationOrderForFilter=XZY)
    Child=(Type=Bone,Name="aim_bags_02_rt")
    Parents="((Item=(Type=Bone,Name=\"dyn_skirt_back_side_01_rt\"),Weight=0.0069),(Item=(Type=Bone,Name=\"dyn_skirt_front_side_03_rt\"),Weight=0.0077),(Item=(Type=Bone,Name=\"dyn_skirt_side_01_rt\"),Weight=0.0129),(Item=(Type=Bone,Name=\"dyn_skirt_front_side_01_rt\"),Weight=0. ...(truncated)"
    Weight=1
   )
   > &RigUnit_ParentConstraint(
    AdvancedSettings=(RotationOrderForFilter=XZY)
    Child=(Type=Bone,Name="aim_bags_01_rt")
    Parents="((Item=(Type=Bone,Name=\"dyn_skirt_back_side_02_rt\"),Weight=0.0006),(Item=(Type=Bone,Name=\"dyn_skirt_front_03_rt\"),Weight=0.0011),(Item=(Type=Bone,Name=\"dyn_skirt_side_01_rt\"),Weight=0.0017),(Item=(Type=Bone,Name=\"dyn_skirt_front_01_rt\"),Weight=0.0042),(Ite ...(truncated)"
    Weight=1
   )
   &RigUnit_ModifyTransforms(
    ItemToModify="((Item=(Type=Bone,Name=\"dyn_bags_lf\"),Transform=(Rotation=(X=0,Y=0,Z=-0.2588,W=0.9659),Translation=(X=0.1366,Y=-1.3864,Z=0),Scale3D=(X=1,Y=1,Z=1))),(Item=(Type=Bone,Name=\"dyn_granade_01_lf\"),Transform=(Rotation=(X=0,Y=0,Z=0,W=-1),Translation=(X=0,Y=-0.9191 ...(truncated)"
    WeightMinimum=0
    Weight=RigVMDispatchNode(BoneName=root, Default=0, Name=deformRig_l_thigh_up_alpha) {template="DISPATCH_RigDispatch_GetAnimAttribute(in Name,in BoneName,in Default,out Value,out Found)"}.Value
    ExecutePin=@RigVMFunction_Sequence_1.B
   )
   > &RigUnit_AimConstraintLocalSpaceOffset(
    Child=(Type=Bone,Name="offset_bags_lf")
    Filter=(bX=True,bY=True,bZ=true)
    Parents=((Item=(Type=Bone,Name="aim_bags_lf"),Weight=1))
    Weight=1
    WorldUp=(Target=(X=0,Y=0,Z=1),Kind=Direction,Space=(Type=Bone,Name="aim_bags_lf"))
   )
   > &RigUnit_AimConstraintLocalSpaceOffset(
    Child=(Type=Bone,Name="offset_granade_01_lf")
    Filter=(bX=True,bY=True,bZ=true)
    Parents=((Item=(Type=Bone,Name="aim_granade_01_lf"),Weight=1))
    Weight=1
    WorldUp=(Target=(X=0,Y=0,Z=1),Kind=Direction,Space=(Type=Bone,Name="aim_granade_01_lf"))
   )
   > &RigUnit_AimConstraintLocalSpaceOffset(
    Child=(Type=Bone,Name="offset_granade_02_lf")
    Filter=(bX=True,bY=True,bZ=true)
    Parents=((Item=(Type=Bone,Name="aim_granade_02_lf"),Weight=1))
    Weight=1
    WorldUp=(Target=(X=0,Y=0,Z=1),Kind=Direction,Space=(Type=Bone,Name="aim_granade_02_lf"))
   )
   > &RigUnit_AimConstraintLocalSpaceOffset(
    Child=(Type=Bone,Name="offset_bags_02_rt")
    Filter=(bX=True,bY=True,bZ=true)
    Parents=((Item=(Type=Bone,Name="aim_bags_02_rt"),Weight=1))
    Weight=1
    WorldUp=(Target=(X=0,Y=0,Z=1),Kind=Direction,Space=(Type=Bone,Name="aim_bags_02_rt"))
   )
   > &RigUnit_AimConstraintLocalSpaceOffset(
    Child=(Type=Bone,Name="offset_bags_01_rt")
    Filter=(bX=true,bY=true,bZ=true)
    Parents=((Item=(Type=Bone,Name="aim_bags_01_rt"),Weight=1))
    Weight=1
    WorldUp=(Target=(X=0,Y=0,Z=1),Kind=Direction,Space=(Type=Bone,Name="aim_bags_01_rt"))
   )
   &RigUnit_ModifyTransforms(
    ItemToModify="((Item=(Type=Bone,Name=\"dyn_bags_02_rt\"),Transform=(Rotation=(X=0,Y=0,Z=0.0872,W=0.9962),Translation=(X=-0.5525,Y=-1.8669,Z=0),Scale3D=(X=1,Y=1,Z=1))),(Item=(Type=Bone,Name=\"dyn_bags_01_rt\"),Transform=(Rotation=(X=0,Y=0,Z=-0.0872,W=0.9962),Translation=(X=0 ...(truncated)"
    WeightMinimum=0
    Weight=RigVMDispatchNode(BoneName=root, Default=0, Name=deformRig_r_thigh_up_alpha) {template="DISPATCH_RigDispatch_GetAnimAttribute(in Name,in BoneName,in Default,out Value,out Found)"}.Value
    ExecutePin=@RigVMFunction_Sequence_1.C
   )
   RigVMFunctionReturnNode(ExecuteContext=@RigVMFunction_Sequence.B)
  ]}
 New Function_ContainedGraph {nodes [
   n1 = RigVMFunctionEntryNode()
   RigVMFunction_MathQuaternionToEuler = &RigVMFunction_MathQuaternionToEuler(
    RotationOrder=ZYX
    Value=&RigUnit_GetRelativeTransformForItem(
     Child=(Type=Bone,Name="hand_l")
     Parent=(Type=Bone,Name="lowerarm_l")
     bChildInitial=False
     bParentInitial=False
     Child.Name=@n1.Child
     Parent.Name=@n1.Parent
    ).RelativeTransform.Rotation
   )
   Remap = &RigVMFunction_MathDoubleRemap(
    TargetMaximum=1
    TargetMinimum=0
    Value=@RigVMFunction_MathQuaternionToEuler.Result.Z
    SourceMaximum=@n1.Source Maximum
    SourceMinimum=@n1.Source Minimum
   )
   > RigVMFunctionReturnNode(Delta=@Remap.Result)
  ]}
 New Function_ContainedGraph {nodes [
   RigVMFunctionEntryNode()
   > RigVMFunction_Sequence = RigVMAggregateNode()
   &RigUnit_ParentConstraint(
    AdvancedSettings=(RotationOrderForFilter=XZY)
    Child=(Type=Bone,Name="aim_bags_lf")
    Parents="((Item=(Type=Bone,Name=\"dyn_skirt_side_03_lf\"),Weight=0.0054),(Item=(Type=Bone,Name=\"dyn_skirt_front_01_lf\"),Weight=0.0059),(Item=(Type=Bone,Name=\"dyn_skirt_front_side_01_lf\"),Weight=0.0164),(Item=(Type=Bone,Name=\"dyn_skirt_side_02_lf\"),Weight=0.0212),(Ite ...(truncated)"
    Weight=1
    ExecutePin=@RigVMFunction_Sequence.A
   )
   > &RigUnit_ParentConstraint(
    AdvancedSettings=(RotationOrderForFilter=XZY)
    Child=(Type=Bone,Name="aim_granade_02_lf")
    Parents="((Item=(Type=Bone,Name=\"dyn_skirt_side_03_lf\"),Weight=0.014),(Item=(Type=Bone,Name=\"dyn_skirt_front_side_02_lf\"),Weight=0.0268),(Item=(Type=Bone,Name=\"dyn_skirt_side_02_lf\"),Weight=0.0268),(Item=(Type=Bone,Name=\"dyn_skirt_back_01_lf\"),Weight=0.0316),(Item= ...(truncated)"
    Weight=1
   )
   > &RigUnit_ParentConstraint(
    AdvancedSettings=(RotationOrderForFilter=XZY)
    Child=(Type=Bone,Name="aim_granade_01_lf")
    Parents="((Item=(Type=Bone,Name=\"dyn_skirt_front_02_lf\"),Weight=0.004),(Item=(Type=Bone,Name=\"dyn_skirt_side_01_lf\"),Weight=0.013),(Item=(Type=Bone,Name=\"dyn_skirt_front_side_01_lf\"),Weight=0.018),(Item=(Type=Bone,Name=\"dyn_skirt_front_side_03_lf\"),Weight=0.0233),( ...(truncated)"
    Weight=1
   )
   > &RigUnit_ParentConstraint(
    AdvancedSettings=(RotationOrderForFilter=XZY)
    Child=(Type=Bone,Name="aim_bags_02_rt")
    Parents="((Item=(Type=Bone,Name=\"dyn_skirt_back_side_01_rt\"),Weight=0.0069),(Item=(Type=Bone,Name=\"dyn_skirt_front_side_03_rt\"),Weight=0.0077),(Item=(Type=Bone,Name=\"dyn_skirt_side_01_rt\"),Weight=0.0129),(Item=(Type=Bone,Name=\"dyn_skirt_front_side_01_rt\"),Weight=0. ...(truncated)"
    Weight=1
   )
   > &RigUnit_ParentConstraint(
    AdvancedSettings=(RotationOrderForFilter=XZY)
    Child=(Type=Bone,Name="aim_bags_01_rt")
    Parents="((Item=(Type=Bone,Name=\"dyn_skirt_back_side_02_rt\"),Weight=0.0212),(Item=(Type=Bone,Name=\"dyn_skirt_front_01_rt\"),Weight=0.0353),(Item=(Type=Bone,Name=\"dyn_skirt_front_03_rt\"),Weight=0.0411),(Item=(Type=Bone,Name=\"dyn_skirt_side_02_rt\"),Weight=0.0582),(Ite ...(truncated)"
    Weight=1
   )
   &RigUnit_ParentConstraint(
    AdvancedSettings=(RotationOrderForFilter=XZY)
    Child=(Type=Bone,Name="base_arm_belt_01_rt")
    Parents=((Item=(Type=Bone,Name="lowerarm_r"),Weight=0.2156),(Item=(Type=Bone,Name="lowerarm_twist_02_r"),Weight=0.2325),(Item=(Type=Bone,Name="lowerarm_twist_01_r"),Weight=0.5519))
    Weight=1
    ExecutePin=@RigVMFunction_Sequence.B
   )
   > &RigUnit_ParentConstraint(
    AdvancedSettings=(RotationOrderForFilter=XZY)
    Child=(Type=Bone,Name="base_arm_belt_02_rt")
    Parents="((Item=(Type=Bone,Name=\"deform_wrist_up_r\"),Weight=0.0196),(Item=(Type=Bone,Name=\"hand_r\"),Weight=0.0269),(Item=(Type=Bone,Name=\"lowerarm_twist_02_r\"),Weight=0.0356),(Item=(Type=Bone,Name=\"lowerarm_r\"),Weight=0.0515),(Item=(Type=Bone,Name=\"dyn_arm_belt_02_ ...(truncated)"
    Weight=1
   )
   RigVMFunctionReturnNode(ExecuteContext=@RigVMFunction_Sequence.C)
  ]}
 New Function_ContainedGraph {nodes [
   n1 = RigVMFunctionEntryNode()
   RigVMFunction_Sequence = &RigVMFunction_Sequence(ExecuteContext=@n1.ExecuteContext)
   RigVMFunctionReturnNode(ExecuteContext=@RigVMFunction_Sequence.B)
   n2 = &RigVMFunction_Sequence(ExecuteContext=@RigVMFunction_Sequence.A)
   RigUnit_GetRelativeTransformForItem = &RigUnit_GetRelativeTransformForItem(Child=(Type=Bone,Name="pelvis"), Parent=(Type=Bone,Name="root"), bChildInitial=False, bParentInitial=False)
   Remap = &RigVMFunction_MathDoubleRemap(
    SourceMaximum=40
    SourceMinimum=70
    TargetMaximum=0
    TargetMinimum=1
    Value=@RigUnit_GetRelativeTransformForItem.RelativeTransform.Translation.Z
   )
   n3 = &RigVMFunction_AnimEvalRichCurve(
    Curve="(EditorCurveData=(Keys=((InterpMode=RCIM_Cubic,TangentMode=RCTM_Break,TangentWeightMode=RCTWM_WeightedNone,Time=0,Value=0,ArriveTangent=0,ArriveTangentWeight=0,LeaveTangent=0,LeaveTangentWeight=0),(InterpMode=RCIM_Cubic,TangentMode=RCTM_Break,TangentWeight ...(truncated)"
    SourceMinimum=0
    TargetMinimum=0
    Value=@Remap.Result
   )
   n4 = &RigVMFunction_MathBoolToFloat(Value=@n1.Crouch)
   Multiply = &RigVMFunction_MathDoubleMul(A=@n3.Result, B=@n4.Result)
   &RigUnit_ModifyTransforms(
    ItemToModify="((Item=(Type=Bone,Name=\"offset_loin_cloth_front_cr\"),Transform=(Rotation=(X=-0.0868,Y=0.0076,Z=-0.0868,W=0.9924),Translation=(X=0,Y=0,Z=0),Scale3D=(X=1,Y=1,Z=1))),(Item=(Type=Bone,Name=\"offset_bags_01_rt\"),Transform=(Rotation=(X=0,Y=0,Z=0,W=1),Translation= ...(truncated)"
    WeightMinimum=0
    Weight=@Multiply.Result
    ExecutePin=@n2.A
   )
   n5 = RigVMDispatchNode(False=0, True=@n1.left thigh, Condition=RigVMVariableNode() {variable=crouch}.Value) {template="DISPATCH_RigVMDispatch_If(in Condition,in True,in False,out Result)"}
   > &RigUnit_ModifyTransforms(
    ItemToModify="((Item=(Type=Bone,Name=\"offset_skirt_front_lf\"),Transform=(Rotation=(X=0,Y=0.0436,Z=0,W=0.999),Translation=(X=0,Y=0,Z=0),Scale3D=(X=1,Y=1,Z=1))),(Item=(Type=Bone,Name=\"dyn_skirt_front_02_lf\"),Transform=(Rotation=(X=0,Y=0.342,Z=0,W=0.9397),Translation=(X=0, ...(truncated)"
    WeightMinimum=0
    Weight=&RigVMFunction_MathDoubleRemap(
     SourceMaximum=0.8
     SourceMinimum=0.5
     TargetMaximum=1
     TargetMinimum=0
     Value=&RigVMFunction_MathDoubleMul(A=@n5.Result, B=@Multiply.Result).Result
    ).Result
   )
   n6 = RigVMDispatchNode(False=0, True=@n1.right thigh, Condition=RigVMVariableNode() {variable=crouch}.Value) {template="DISPATCH_RigVMDispatch_If(in Condition,in True,in False,out Result)"}
   > &RigUnit_ModifyTransforms(
    ItemToModify="((Item=(Type=Bone,Name=\"offset_skirt_front_rt\"),Transform=(Rotation=(X=0,Y=0.061,Z=0,W=0.9981),Translation=(X=0,Y=0,Z=0),Scale3D=(X=1,Y=1,Z=1))),(Item=(Type=Bone,Name=\"dyn_skirt_front_02_rt\"),Transform=(Rotation=(X=0,Y=0.1305,Z=0,W=0.9914),Translation=(X=0 ...(truncated)"
    WeightMinimum=0
    Weight=&RigVMFunction_MathDoubleRemap(
     SourceMaximum=0.8
     SourceMinimum=0.5
     TargetMaximum=1
     TargetMinimum=0
     Value=&RigVMFunction_MathDoubleMul(A=@n6.Result, B=@Multiply.Result).Result
    ).Result
   )
   n7 = &RigVMFunction_AnimEvalRichCurve(
    Curve="(EditorCurveData=(Keys=((InterpMode=RCIM_Cubic,TangentMode=RCTM_Break,TangentWeightMode=RCTWM_WeightedNone,Time=0,Value=0,ArriveTangent=0,ArriveTangentWeight=0,LeaveTangent=0,LeaveTangentWeight=0),(InterpMode=RCIM_Cubic,TangentMode=RCTM_Break,TangentWeight ...(truncated)"
    SourceMinimum=0
    TargetMaximum=0
    TargetMinimum=1
    Value=@Remap.Result
   )
   n8 = &RigVMFunction_MathDoubleMul(A=@n7.Result, B=@n4.Result)
   &RigUnit_ModifyTransforms(
    ItemToModify="((Item=(Type=Bone,Name=\"offset_loin_cloth_front_cr\"),Transform=(Rotation=(X=0,Y=0.1736,Z=0,W=0.9848),Translation=(X=0,Y=0,Z=0),Scale3D=(X=1,Y=1,Z=1))),(Item=(Type=Bone,Name=\"offset_bags_lf\"),Transform=(Rotation=(X=0,Y=0,Z=0.1736,W=0.9848),Translation=(X=0, ...(truncated)"
    WeightMinimum=0
    Weight=@n8.Result
    ExecutePin=@n2.B
   )
   > &RigUnit_ModifyTransforms(
    ItemToModify="((Item=(Type=Bone,Name=\"base_belt_01_01\"),Transform=(Rotation=(X=0.0302,Y=0.171,Z=0.171,W=0.9698),Translation=(X=0,Y=0,Z=0),Scale3D=(X=1,Y=1,Z=1))),(Item=(Type=Bone,Name=\"base_belt_02_01\"),Transform=(Rotation=(X=0,Y=0.1736,Z=0,W=0.9848),Translation=(X=0,Y= ...(truncated)"
    WeightMinimum=0
    Weight=@n8.Result
   )
   > &RigUnit_ModifyTransforms(
    ItemToModify="((Item=(Type=Bone,Name=\"offset_skirt_back_side_lf\"),Transform=(Rotation=(X=0,Y=0.0872,Z=0,W=0.9962),Translation=(X=0,Y=0,Z=0),Scale3D=(X=1,Y=1,Z=1))),(Item=(Type=Bone,Name=\"dyn_skirt_back_side_02_lf\"),Transform=(Rotation=(X=0,Y=0.1736,Z=0,W=0.9848),Transla ...(truncated)"
    WeightMinimum=0
    Weight=@n8.Result
   )
   > &RigUnit_ModifyTransforms(
    ItemToModify="((Item=(Type=Bone,Name=\"offset_skirt_front_side_lf\"),Transform=(Rotation=(X=0,Y=0.0872,Z=0,W=0.9962),Translation=(X=0,Y=0,Z=0),Scale3D=(X=1,Y=1,Z=1))),(Item=(Type=Bone,Name=\"dyn_skirt_front_side_02_lf\"),Transform=(Rotation=(X=0,Y=0.1736,Z=0,W=0.9848),Trans ...(truncated)"
    WeightMinimum=0
    Weight=@n8.Result
   )
   > &RigUnit_ModifyTransforms(
    ItemToModify="((Item=(Type=Bone,Name=\"offset_skirt_side_lf\"),Transform=(Rotation=(X=0.056,Y=0.0668,Z=-0.6403,W=0.7631),Translation=(X=0,Y=0,Z=0),Scale3D=(X=1,Y=1,Z=1))),(Item=(Type=Bone,Name=\"dyn_skirt_side_02_lf\"),Transform=(Rotation=(X=0,Y=0.1736,Z=0,W=0.9848),Transla ...(truncated)"
    WeightMinimum=0
    Weight=@n8.Result
   )
   > &RigUnit_ModifyTransforms(
    ItemToModify="((Item=(Type=Bone,Name=\"offset_skirt_front_lf\"),Transform=(Rotation=(X=0,Y=0,Z=0,W=1),Translation=(X=0,Y=0,Z=0),Scale3D=(X=1,Y=1,Z=1))),(Item=(Type=Bone,Name=\"dyn_skirt_front_02_lf\"),Transform=(Rotation=(X=0,Y=0,Z=0,W=1),Translation=(X=0,Y=0,Z=0),Scale3D=( ...(truncated)"
    WeightMinimum=0
    Weight=@n8.Result
   )
   > &RigUnit_ModifyTransforms(
    ItemToModify="((Item=(Type=Bone,Name=\"offset_skirt_back_side_rt\"),Transform=(Rotation=(X=0,Y=0,Z=0,W=1),Translation=(X=0,Y=0,Z=0),Scale3D=(X=1,Y=1,Z=1))),(Item=(Type=Bone,Name=\"dyn_skirt_back_side_02_rt\"),Transform=(Rotation=(X=0,Y=0,Z=0,W=1),Translation=(X=0,Y=0,Z=0),S ...(truncated)"
    WeightMinimum=0
    Weight=@n8.Result
   )
   > &RigUnit_ModifyTransforms(
    ItemToModify="((Item=(Type=Bone,Name=\"offset_skirt_side_rt\"),Transform=(Rotation=(X=-0.0302,Y=-0.171,Z=-0.171,W=0.9698),Translation=(X=0,Y=0,Z=0),Scale3D=(X=1,Y=1,Z=1))),(Item=(Type=Bone,Name=\"dyn_skirt_back_side_02_rt\"),Transform=(Rotation=(X=0,Y=0,Z=0,W=1),Translation ...(truncated)"
    WeightMinimum=0
    Weight=@n8.Result
   )
   > &RigUnit_ModifyTransforms(
    ItemToModify="((Item=(Type=Bone,Name=\"offset_skirt_front_side_rt\"),Transform=(Rotation=(X=0,Y=0.0872,Z=0,W=0.9962),Translation=(X=0,Y=0,Z=0),Scale3D=(X=1,Y=1,Z=1))),(Item=(Type=Bone,Name=\"dyn_skirt_back_side_02_rt\"),Transform=(Rotation=(X=0,Y=0,Z=0,W=1),Translation=(X=0 ...(truncated)"
    WeightMinimum=0
    Weight=@n8.Result
   )
  ]}
 RigVMFunctionLibrary {nodes [
   Bag = RigVMCollapseNode(ExecuteContext=())
   Controls = RigVMCollapseNode(Debug=False, ExecuteContext=())
   "Crouch Skirt " = RigVMCollapseNode(Crouch=False, ExecuteContext=())
   Delta Rotation = RigVMCollapseNode(ExecuteContext=())
   Skirt = RigVMCollapseNode(ExecuteContext=())
   Surface = RigVMCollapseNode(ExecuteContext=())
  ]}
 RigVMFunction_Sequence_1_ContainedGraph {nodes [
   n1 = RigVMFunctionEntryNode()
   RigVMFunction_Sequence = &RigVMFunction_Sequence(ExecuteContext=@n1.ExecuteContext)
   n2 = &RigVMFunction_Sequence(ExecuteContext=@RigVMFunction_Sequence.B)
   n3 = &RigVMFunction_Sequence(ExecuteContext=@n2.B)
   n4 = &RigVMFunction_Sequence(ExecuteContext=@n3.B)
   RigVMFunctionReturnNode(A=@RigVMFunction_Sequence.A, B=@n2.A, C=@n3.A, D=@n4.A, E=@n4.B)
  ]}
 RigVMFunction_Sequence_1_ContainedGraph {nodes [
   n1 = RigVMFunctionEntryNode()
   RigVMFunction_Sequence = &RigVMFunction_Sequence(ExecuteContext=@n1.ExecuteContext)
   n2 = &RigVMFunction_Sequence(ExecuteContext=@RigVMFunction_Sequence.B)
   RigVMFunctionReturnNode(A=@RigVMFunction_Sequence.A, B=@n2.A, C=@n2.B)
  ]}
 RigVMFunction_Sequence_2_ContainedGraph {nodes [
   n1 = RigVMFunctionEntryNode()
   n2 = &RigVMFunction_Sequence(ExecuteContext=@n1.ExecuteContext)
   n3 = &RigVMFunction_Sequence(ExecuteContext=@n2.B)
   n4 = &RigVMFunction_Sequence(ExecuteContext=@n3.B)
   RigVMFunctionReturnNode(A=@n2.A, B=@n3.A, C=@n4.A, D=@n4.B)
  ]}
 RigVMFunction_Sequence_2_ContainedGraph {nodes [
   n1 = RigVMFunctionEntryNode()
   n2 = &RigVMFunction_Sequence(ExecuteContext=@n1.ExecuteContext)
   n3 = &RigVMFunction_Sequence(ExecuteContext=@n2.B)
   RigVMFunctionReturnNode(A=@n2.A, B=@n3.A, C=@n3.B)
  ]}
 RigVMFunction_Sequence_3_ContainedGraph {nodes [
   n1 = RigVMFunctionEntryNode()
   n2 = &RigVMFunction_Sequence(ExecuteContext=@n1.ExecuteContext)
   n3 = &RigVMFunction_Sequence(ExecuteContext=@n2.B)
   RigVMFunctionReturnNode(B=@n3.A, C=@n3.B, A=@n2.A)
  ]}
 RigVMFunction_Sequence_4_ContainedGraph {nodes [
   n1 = RigVMFunctionEntryNode()
   n2 = &RigVMFunction_Sequence(ExecuteContext=@n1.ExecuteContext)
   n3 = &RigVMFunction_Sequence(ExecuteContext=@n2.B)
   RigVMFunctionReturnNode(B=@n3.A, C=@n3.B, A=@n2.A)
  ]}
 RigVMModel {nodes [
   RigUnit_BeginExecution = &RigUnit_BeginExecution()
   RigVMFunction_Sequence = RigVMAggregateNode(ExecuteContext=@RigUnit_BeginExecution.ExecutePin)
   Controls = RigVMFunctionReferenceNode(Debug=false, ExecuteContext=@RigVMFunction_Sequence.A)
   FN_DeformRig_FNC = RigVMFunctionReferenceNode(
    DeformRig_ Curve="(deformRig_l_upperarm_up_alpha_72_#1=False,deformRig_r_upperarm_up_alpha_73_#2=False,deformRig_l_upperarm_fr_alpha_74_#3=False,deformRig_r_upperarm_fr_alpha_75_CB51B8 ...(truncated)"
    DeformRig_Bool="(upperarm_1_#4=True,elbow_3_#5=True,wrist_7_#6=True,groin_8_#7=True,glute_17_#8=True,knee_9_40C408324CA261 ...(truncated)"
    DeformRig_Debug="(upperarm_debug=False,wrist_debug=False,lat_debug=False,groinGlute_debug=False,upperarm_debug_18_#4=False,wrist_debug_20_#6=False,groinGlute_debug_27_#7=False,lat_deb ...(truncated)"
    DeformRig_Value="(upperarm_up_18_#9=(Rotation=(X=0.0,Y=-0.0,Z=0.0,W=1.0),Translation=(X=0.0,Y=0.0,Z=8.0),Scale3D=(X=0.70,Y=1.0,Z=1.0)),upperarm_fwd_fr_19_#10=(Rotation=(X=0.0,Y=-0.0,Z=0.0,W=1.0),Translation=(X=0.0, ...(truncated)"
    ExecuteContext=@RigVMFunction_Sequence.B
   )
   Skirt = RigVMFunctionReferenceNode(ExecuteContext=@RigVMFunction_Sequence.C)
   > Bag = RigVMFunctionReferenceNode()
   > Surface = RigVMFunctionReferenceNode()
   > "Crouch Skirt " = RigVMFunctionReferenceNode(left thigh=@Skirt.left thigh, right thigh=@Skirt.right thigh, Crouch=RigVMVariableNode() {variable=crouch}.Value)
   > n1 = &RigUnit_SphericalPoseReader(
    ActiveRegionScaleFactors=(PositiveWidth=1,NegativeWidth=1,PositiveHeight=1,NegativeHeight=1)
    ActiveRegionSize=0.25
    Debug=(bDrawDebug=false,bDraw2D=False,bDrawLocalAxes=False,DebugScale=25,DebugSegments=20,DebugThickness=0.25)
    DriverAxis=(X=1,Y=0,Z=0)
    DriverItem=(Type=Bone,Name="upperarm_l")
    FalloffRegionScaleFactors=(PositiveWidth=1.5,NegativeWidth=1,PositiveHeight=1,NegativeHeight=1)
    FalloffSize=0.4
    FlipHeightScaling=False
    FlipWidthScaling=False
    OptionalParentItem=(Type=Bone,Name="None")
    RotationOffset=(X=0,Y=-5,Z=0)
   )
   > &RigUnit_ModifyTransforms(
    ItemToModify=((Item=(Type=Bone,Name="shoulderpad_l"),Transform=(Rotation=(X=0.0872,Y=0,Z=0,W=0.9962),Translation=(X=4,Y=-1,Z=1),Scale3D=(X=1,Y=1,Z=1))))
    WeightMinimum=0
    Weight=@n1.OutputParam
   )
   > n2 = &RigUnit_SphericalPoseReader(
    ActiveRegionScaleFactors=(PositiveWidth=1,NegativeWidth=1,PositiveHeight=1,NegativeHeight=1)
    ActiveRegionSize=0.05
    Debug=(bDrawDebug=false,bDraw2D=False,bDrawLocalAxes=False,DebugScale=25,DebugSegments=20,DebugThickness=0.25)
    DriverAxis=(X=0,Y=1,Z=0)
    DriverItem=(Type=Bone,Name="spine_02")
    FalloffRegionScaleFactors=(PositiveWidth=1,NegativeWidth=1,PositiveHeight=1,NegativeHeight=1)
    FalloffSize=0.15
    FlipHeightScaling=False
    FlipWidthScaling=False
    OptionalParentItem=(Type=Bone,Name="None")
    RotationOffset=(X=-245,Y=0,Z=0)
   )
   > &RigUnit_ModifyTransforms(
    ItemToModify="((Item=(Type=Bone,Name=\"offset_skirt_front_rt\"),Transform=(Rotation=(X=0,Y=-0.0436,Z=0,W=0.999),Translation=(X=0,Y=0,Z=0),Scale3D=(X=1,Y=1,Z=1))),(Item=(Type=Bone,Name=\"offset_skirt_front_side_rt\"),Transform=(Rotation=(X=0,Y=-0.0436,Z=0,W=0.999),Translatio ...(truncated)"
    WeightMinimum=0
    Weight=@n2.OutputParam
   )
  ]}
]
hierarchy [
 root {loc=<0 0 0>, rot=<0 0 0 1>}
 pelvis {parent=root, loc=<0 -0.5301 98.0571>, rot=<0.0127 0.707 -0.0127 -0.707>}
 spine_01 {parent=pelvis, loc=<6.1208 -0.9328 0>, rot=<0 0 0.034 -0.9994>}
 spine_02 {parent=spine_01, loc=<6.5997 -0.2095 0>, rot=<0 0 0.0238 -0.9997>}
 spine_03 {parent=spine_02, loc=<6.4914 -0.1237 0>, rot=<0 0 0.0149 -0.9999>}
 spine_04 {parent=spine_03, loc=<6.9598 0.0794 0>, rot=<0 0 -0.0132 -0.9999>}
 spine_05 {parent=spine_04, loc=<6.9987 0 0>, rot=<0 0 -0.0015 -1>}
 neck_01 {parent=spine_05, loc=<11.3138 0.1529 0>, rot=<0 0 -0.2003 -0.9797>}
 neck_02 {parent=neck_01, loc=<4.5627 0 0>, rot=<0 0 0.0504 -0.9987>}
 head {parent=neck_02, loc=<5.3914 0 0>, rot=<0 0 0.115 -0.9934>}
 clavicle_l {parent=spine_05, loc=<5.4831 -0.1053 -3.2946>, rot=<0.0465 -0.7289 -0.009 -0.6829>}
 upperarm_l {parent=clavicle_l, loc=<9.8658 -1.2136 -0.4105>, rot=<0.0164 -0.3245 -0.0269 -0.9454>}
 lowerarm_l {parent=upperarm_l, loc=<26.465 -0.8345 0.259>, rot=<0.0036 0.0161 -0.2298 -0.9731>}
 hand_l {parent=lowerarm_l, loc=<22.7355 -0.3203 -0.6105>, rot=<-0.6784 -0.0223 0.0735 -0.7307>}
 thumb_01_l {parent=hand_l, loc=<1.3118 -0.3191 -2.459>, rot=<0.6546 -0.4042 0.0664 -0.6354>}
 thumb_02_l {parent=thumb_01_l, loc=<5.9336 -0.0448 0.0401>, rot=<-0.0037 0.0213 0.298 -0.9543>}
 thumb_03_l {parent=thumb_02_l, loc=<2.2223 0.3682 -0.0907>, rot=<0.0118 -0.013 0.0066 -0.9998>}
 index_metacarpal_l {parent=hand_l, loc=<1.6535 0.0183 -2.96>, rot=<0.0158 -0.0124 -0.0173 -0.9996>}
 index_01_l {parent=index_metacarpal_l, loc=<8.3008 0.0357 -0.0524>, rot=<-0.0044 0.0485 0.0924 -0.9945>}
 index_02_l {parent=index_01_l, loc=<3.3587 -0.3004 0.0452>, rot=<-0.002 0.0027 0.3225 -0.9466>}
 index_03_l {parent=index_02_l, loc=<2.5578 -0.123 -0.0252>, rot=<-0.002 0.0067 0.2558 -0.9667>}
 middle_metacarpal_l {parent=hand_l, loc=<2.1856 -0.1725 -0.5984>, rot=<0.006 0.0168 -0.0329 -0.9993>}
 middle_01_l {parent=middle_metacarpal_l, loc=<7.3275 -0.2447 0.0129>, rot=<-0.0208 0.0042 0.0858 -0.9961>}
 middle_02_l {parent=middle_01_l, loc=<4.0123 0.1401 0.1175>, rot=<-0.0218 0.0258 0.3512 -0.9357>}
 middle_03_l {parent=middle_02_l, loc=<2.6077 0.0534 0.1373>, rot=<0.0082 0.0188 0.1708 -0.9851>}
 ring_metacarpal_l {parent=hand_l, loc=<2.2074 -0.3625 1.3638>, rot=<-0.0284 0.0722 -0.0458 -0.9959>}
 ring_01_l {parent=ring_metacarpal_l, loc=<7.0295 -0.6197 -0.1446>, rot=<-0.0489 -0.0252 0.1892 -0.9804>}
 ring_02_l {parent=ring_01_l, loc=<3.6174 0.4647 -0.1027>, rot=<0 0 0.2806 -0.9598>}
 ring_03_l {parent=ring_02_l, loc=<2.4152 0.114 0.0199>, rot=<0 0 0.1189 -0.9929>}
 pinky_metacarpal_l {parent=hand_l, loc=<1.7616 -0.9753 3.3814>, rot=<-0.0397 0.0814 -0.069 -0.9935>}
 pinky_01_l {parent=pinky_metacarpal_l, loc=<7.1709 -0.4831 -0.1094>, rot=<-0.0654 -0.0035 0.2242 -0.9723>}
 pinky_02_l {parent=pinky_01_l, loc=<2.8274 0.0893 -0.1043>, rot=<0.0037 0.0011 0.2854 -0.9584>}
 pinky_03_l {parent=pinky_02_l, loc=<2.2046 0.0307 -0.0078>, rot=<-0.0002 0.0098 0.1445 -0.9895>}
 deform_wrist_dn_base_l {parent=hand_l, loc=<0 -1.5326 0>, rot=<0 0 0 -1>}
 deform_wrist_dn_l {parent=deform_wrist_dn_base_l, loc=<0 0 0>, rot=<0 0 0 1>}
 deform_wrist_up_base_l {parent=hand_l, loc=<0 2.3009 0>, rot=<0 0 0 -1>}
 deform_wrist_up_l {parent=deform_wrist_up_base_l, loc=<0 0 0>, rot=<0 0 0 1>}
 lowerarm_twist_02_l {parent=lowerarm_l, loc=<10.6707 0 0>, rot=<0 0 0 1>}
 lowerarm_twist_01_l {parent=lowerarm_l, loc=<18.316 0 0>, rot=<0 0 0 1>}
 deform_elbow_in_base_l {parent=lowerarm_l, loc=<0.1834 3.6062 -0.2572>, rot=<0 0 0 -1>}
 deform_elbow_in_l {parent=deform_elbow_in_base_l, loc=<0 0 0>, rot=<0 0 0 1>}
 deform_elbow_out_base_l {parent=lowerarm_l, loc=<0 -4 0>, rot=<0 0 0 1>}
 deform_elbow_out_l {parent=deform_elbow_out_base_l, loc=<0 0 0>, rot=<0 0 0 1>}
 upperarm_twist_01_l {parent=upperarm_l, loc=<7.2135 -0.1134 0>, rot=<0 0 0 1>}
 deform_upperarm_bk_base_l {parent=upperarm_twist_01_l, loc=<-6.4053 -4.3746 0.314>, rot=<0 0 0 -1>}
 deform_upperarm_bk_l {parent=deform_upperarm_bk_base_l, loc=<0 0 0>, rot=<0 0 0 1>}
 deform_upperarm_fr_base_l {parent=upperarm_twist_01_l, loc=<-5.6312 5.092 0.3354>, rot=<0 0 0 -1>}
 deform_upperarm_fr_l {parent=deform_upperarm_fr_base_l, loc=<0 0 0>, rot=<0 0 0 1>}
 deform_upperarm_up_base_l {parent=upperarm_twist_01_l, loc=<-8.891 0.4825 3.9517>, rot=<0 0 0 -1>}
 deform_upperarm_up_l {parent=deform_upperarm_up_base_l, loc=<0 0 0>, rot=<0 0 0 1>}
 shoulderpad_l {parent=upperarm_twist_01_l, loc=<0 0 0>, rot=<0 0 0 1>}
 upperarm_twist_02_l {parent=upperarm_l, loc=<19.7998 -0.1134 0>, rot=<0 0 0 1>}
 clavicle_r {parent=spine_05, loc=<5.4831 -0.1053 3.2946>, rot=<0.7289 0.0465 -0.6829 0.009>}
 upperarm_r {parent=clavicle_r, loc=<-9.8658 1.2136 0.4105>, rot=<0.0164 -0.3245 -0.0269 -0.9454>}
 lowerarm_r {parent=upperarm_r, loc=<-26.465 0.8345 -0.259>, rot=<0.0036 0.0161 -0.2298 -0.9731>}
 hand_r {parent=lowerarm_r, loc=<-22.7355 0.3203 0.6106>, rot=<-0.6784 -0.0223 0.0735 -0.7307>}
 thumb_01_r {parent=hand_r, loc=<-1.3118 0.3191 2.459>, rot=<0.6546 -0.4042 0.0664 -0.6354>}
 thumb_02_r {parent=thumb_01_r, loc=<-5.9336 0.0448 -0.0401>, rot=<-0.0037 0.0213 0.298 -0.9543>}
 thumb_03_r {parent=thumb_02_r, loc=<-2.2223 -0.3682 0.0907>, rot=<0.0118 -0.013 0.0066 -0.9998>}
 index_metacarpal_r {parent=hand_r, loc=<-1.6535 -0.0183 2.96>, rot=<0.0158 -0.0124 -0.0173 -0.9996>}
 index_01_r {parent=index_metacarpal_r, loc=<-8.3008 -0.0357 0.0524>, rot=<-0.0044 0.0485 0.0924 -0.9945>}
 index_02_r {parent=index_01_r, loc=<-3.3587 0.3004 -0.0452>, rot=<-0.002 0.0027 0.3225 -0.9466>}
 index_03_r {parent=index_02_r, loc=<-2.5578 0.123 0.0252>, rot=<-0.002 0.0067 0.2558 -0.9667>}
 middle_metacarpal_r {parent=hand_r, loc=<-2.1856 0.1725 0.5984>, rot=<0.006 0.0168 -0.0329 -0.9993>}
 middle_01_r {parent=middle_metacarpal_r, loc=<-7.3275 0.2447 -0.0129>, rot=<-0.0208 0.0042 0.0858 -0.9961>}
 middle_02_r {parent=middle_01_r, loc=<-4.0123 -0.1401 -0.1175>, rot=<-0.0218 0.0258 0.3512 -0.9357>}
 middle_03_r {parent=middle_02_r, loc=<-2.6077 -0.0534 -0.1373>, rot=<0.0082 0.0188 0.1708 -0.9851>}
 ring_metacarpal_r {parent=hand_r, loc=<-2.2074 0.3625 -1.3638>, rot=<-0.0284 0.0722 -0.0458 -0.9959>}
 ring_01_r {parent=ring_metacarpal_r, loc=<-7.0295 0.6197 0.1446>, rot=<-0.0489 -0.0252 0.1892 -0.9804>}
 ring_02_r {parent=ring_01_r, loc=<-3.6174 -0.4647 0.1027>, rot=<0 0 0.2806 -0.9598>}
 ring_03_r {parent=ring_02_r, loc=<-2.4152 -0.114 -0.0199>, rot=<0 0 0.1189 -0.9929>}
 pinky_metacarpal_r {parent=hand_r, loc=<-1.7616 0.9753 -3.3814>, rot=<-0.0397 0.0814 -0.069 -0.9935>}
 pinky_01_r {parent=pinky_metacarpal_r, loc=<-7.1709 0.4831 0.1094>, rot=<-0.0654 -0.0035 0.2242 -0.9723>}
 pinky_02_r {parent=pinky_01_r, loc=<-2.8274 -0.0893 0.1043>, rot=<0.0037 0.0011 0.2854 -0.9584>}
 pinky_03_r {parent=pinky_02_r, loc=<-2.2046 -0.0307 0.0078>, rot=<-0.0002 0.0098 0.1445 -0.9895>}
 deform_wrist_dn_base_r {parent=hand_r, loc=<-0.0001 1.5328 0>, rot=<0 0 0 -1>}
 deform_wrist_dn_r {parent=deform_wrist_dn_base_r, loc=<0 0 0>, rot=<0 0 0 1>}
 deform_wrist_up_base_r {parent=hand_r, loc=<-0.0002 -2.3007 -0.0001>, rot=<0 0 0 -1>}
 deform_wrist_up_r {parent=deform_wrist_up_base_r, loc=<0 0 0>, rot=<0 0 0 1>}
 lowerarm_twist_02_r {parent=lowerarm_r, loc=<-10.6707 0 0>, rot=<0 0 0 -1>}
 lowerarm_twist_01_r {parent=lowerarm_r, loc=<-18.316 0 0>, rot=<0 0 0 -1>}
 base_arm_belt_01_rt {parent=lowerarm_r, loc=<-11.9034 1.6842 -4.021>, rot=<-0.6944 0.0923 0.7083 -0.0872>}
 dyn_arm_belt_01_rt {parent=base_arm_belt_01_rt, loc=<0 0 0>, rot=<0 0 0 1>}
 base_arm_belt_02_rt {parent=lowerarm_r, loc=<-17.4385 2.1784 -3.0281>, rot=<-0.6964 0.0758 0.7102 -0.0704>}
 dyn_arm_belt_02_rt {parent=base_arm_belt_02_rt, loc=<0 0 0>, rot=<0 0 0 1>}
 deform_elbow_in_base_r {parent=lowerarm_r, loc=<-0.1834 -3.6062 0.2572>, rot=<0 0 0 -1>}
 deform_elbow_in_r {parent=deform_elbow_in_base_r, loc=<0 0 0>, rot=<0 0 0 1>}
 deform_elbow_out_base_r {parent=lowerarm_r, loc=<0 4 -0.0001>, rot=<0 0 0 -1>}
 deform_elbow_out_r {parent=deform_elbow_out_base_r, loc=<0 0 0>, rot=<0 0 0 1>}
 upperarm_twist_01_r {parent=upperarm_r, loc=<-7.2135 0.1134 0>, rot=<0 0 0 -1>}
 deform_upperarm_bk_base_r {parent=upperarm_twist_01_r, loc=<6.4053 4.3746 -0.314>, rot=<0 0 0 -1>}
 deform_upperarm_bk_r {parent=deform_upperarm_bk_base_r, loc=<0 0 0>, rot=<0 0 0 1>}
 deform_upperarm_fr_base_r {parent=upperarm_twist_01_r, loc=<5.6312 -5.092 -0.3354>, rot=<0 0 0 -1>}
 deform_upperarm_fr_r {parent=deform_upperarm_fr_base_r, loc=<0 0 0>, rot=<0 0 0 1>}
 deform_upperarm_up_base_r {parent=upperarm_twist_01_r, loc=<8.891 -0.4825 -3.9517>, rot=<0 0 0 -1>}
 deform_upperarm_up_r {parent=deform_upperarm_up_base_r, loc=<0 0 0>, rot=<0 0 0 1>}
 upperarm_twist_02_r {parent=upperarm_r, loc=<-19.7998 0.1134 0>, rot=<0 0 0 -1>}
 attach_backpack {parent=spine_05, loc=<-0.1314 -13.4712 0.0002>, rot=<-0.0093 -0.707 0.0093 -0.707>}
 Backpack_BR {parent=attach_backpack, type=Null, loc=<0 0 0>, rot=<0 0 0 1>}
 attach_cape {parent=spine_05, loc=<9.0129 -11.7128 0.0002>, rot=<0.0616 -0.7044 -0.0616 -0.7044>}
 cape {parent=attach_cape, type=Null, loc=<0 0 0>, rot=<0 0 0 1>}
 deform_pec_aim_base_l {parent=spine_04, loc=<3.2456 9.3309 -1.2237>, rot=<0.0603 -0.4033 0.3318 -0.8507>}
 deform_pec_aim_l {parent=deform_pec_aim_base_l, loc=<0 0 0>, rot=<0 0 0 1>}
 deform_pec_base_l {parent=deform_pec_aim_l, loc=<9.3665 0 0>, rot=<0 0 0 1>}
 deform_pec_l {parent=deform_pec_base_l, loc=<0 0 0>, rot=<0 0 0 1>}
 deform_pec_aim_base_r {parent=spine_04, loc=<3.2452 9.3309 1.2237>, rot=<-0.4033 -0.0603 0.8507 0.3318>}
 deform_pec_aim_r {parent=deform_pec_aim_base_r, loc=<0 0 0>, rot=<0 0 0 1>}
 deform_pec_base_r {parent=deform_pec_aim_r, loc=<-9.3667 -0.0003 -0.0004>, rot=<0 0 0 1>}
 deform_pec_r {parent=deform_pec_base_r, loc=<0 0 0>, rot=<0 0 0 1>}
 thigh_l {parent=pelvis, loc=<-1.0855 -0.2109 -8.7053>, rot=<0.0009 0.0355 0.0155 -0.9993>}
 calf_l {parent=thigh_l, loc=<-42.5284 -0.0019 -0.7297>, rot=<0.0006 -0.0161 -0.0391 -0.9991>}
 foot_l {parent=calf_l, loc=<-46.7401 -2.5014 -0.592>, rot=<0.0427 -0.0238 0.0353 -0.9982>}
 ball_l {parent=foot_l, loc=<-3.5658 10.8066 -0.1544>, rot=<0.0001 -0.0001 -0.7063 -0.7079>}
 calf_twist_02_l {parent=calf_l, loc=<-19.1546 -1.7943 -0.2638>, rot=<0.0237 0.0095 -0.0035 -0.9997>}
 calf_twist_01_l {parent=calf_l, loc=<-35.8483 -1.9197 -0.5794>, rot=<0.0237 0.0095 -0.0035 -0.9997>}
 deform_knee_in_base_l {parent=calf_l, loc=<0.8031 -5.0231 0.2456>, rot=<0 0 0 -1>}
 deform_knee_in_l {parent=deform_knee_in_base_l, loc=<0 0 0>, rot=<0 0 0 1>}
 deform_knee_out_base_l {parent=calf_l, loc=<0 4 0>, rot=<0 0 0 -1>}
 deform_knee_out_l {parent=deform_knee_out_base_l, loc=<0 0 0>, rot=<0 0 0 1>}
 thigh_twist_01_l {parent=thigh_l, loc=<-27.952 0 0>, rot=<0.0013 0 0 -1>}
 deform_groin_base_l {parent=thigh_l, loc=<0.4801 5.6776 -0.2486>, rot=<0 0 0 -1>}
 deform_groin_l {parent=deform_groin_base_l, loc=<0 0 0>, rot=<0 0 0 1>}
 thigh_r {parent=pelvis, loc=<-1.0855 -0.2109 8.7053>, rot=<0.0355 -0.0009 0.9993 0.0155>}
 calf_r {parent=thigh_r, loc=<42.5284 0.0019 0.7297>, rot=<0.0006 -0.0161 -0.0391 -0.9991>}
 foot_r {parent=calf_r, loc=<46.7401 2.5014 0.592>, rot=<0.0427 -0.0238 0.0353 -0.9982>}
 ball_r {parent=foot_r, loc=<3.5658 -10.8066 0.1544>, rot=<0.0001 -0.0001 -0.7063 -0.7079>}
 calf_twist_02_r {parent=calf_r, loc=<19.1546 1.7943 0.2638>, rot=<0.0237 0.0095 -0.0035 -0.9997>}
 calf_twist_01_r {parent=calf_r, loc=<35.8483 1.9197 0.5794>, rot=<0.0237 0.0095 -0.0035 -0.9997>}
 deform_knee_in_base_r {parent=calf_r, loc=<-0.8031 5.0231 -0.2457>, rot=<0 0 0 -1>}
 deform_knee_in_r {parent=deform_knee_in_base_r, loc=<0 0 0>, rot=<0 0 0 1>}
 deform_knee_out_base_r {parent=calf_r, loc=<0 -4 0>, rot=<0 0 0 -1>}
 deform_knee_out_r {parent=deform_knee_out_base_r, loc=<0 0 0>, rot=<0 0 0 1>}
 thigh_twist_01_r {parent=thigh_r, loc=<27.952 0 0>, rot=<0.0013 0 0 -1>}
 deform_groin_base_r {parent=thigh_r, loc=<-0.4801 -5.6776 0.2486>, rot=<0 0 0 -1>}
 deform_groin_r {parent=deform_groin_base_r, loc=<0 0 0>, rot=<0 0 0 1>}
 base_loin_cloth_front_rt {parent=pelvis, loc=<7.6594 6.7198 5.6225>, rot=<-0.5089 -0.491 -0.491 -0.5089>}
 offset_loin_cloth_front_rt {parent=base_loin_cloth_front_rt, loc=<0 0 0>, rot=<0 0 0 1>}
 dyn_loin_cloth_front_01_rt {parent=offset_loin_cloth_front_rt, loc=<0 0 0>, rot=<0 -0.5532 0 -0.833>}
 dyn_loin_cloth_front_02_rt {parent=dyn_loin_cloth_front_01_rt, loc=<7.4296 0 0>, rot=<0 -0.1245 0 -0.9922>}
 dyn_loin_cloth_front_03_rt {parent=dyn_loin_cloth_front_02_rt, loc=<7.4296 0 0>, rot=<0 0 0 1>}
 base_loin_cloth_front_lf {parent=pelvis, loc=<7.6594 6.7197 -5.623>, rot=<-0.5089 -0.491 -0.491 -0.5089>}
 auto_loin_cloth_front_lf {parent=base_loin_cloth_front_lf, loc=<0 0 0>, rot=<0 0 0 1>}
 offset_loin_cloth_front_lf {parent=auto_loin_cloth_front_lf, loc=<0 0 0>, rot=<0 0 0 1>}
 dyn_loin_cloth_front_01_lf {parent=offset_loin_cloth_front_lf, loc=<0 0 0>, rot=<0 -0.5532 0 -0.833>}
 dyn_loin_cloth_front_02_lf {parent=dyn_loin_cloth_front_01_lf, loc=<7.4296 0 0>, rot=<0 -0.1245 0 -0.9922>}
 dyn_loin_cloth_front_03_lf {parent=dyn_loin_cloth_front_02_lf, loc=<7.4296 0 0>, rot=<0 0 0 1>}
 base_skirt_front_lf {parent=pelvis, loc=<8 7.1032 -3.4928>, rot=<-0.5089 -0.491 -0.491 -0.5089>}
 auto_skirt_front_lf {parent=base_skirt_front_lf, loc=<0 0 0>, rot=<0 0 0 1>}
 offset_skirt_front_lf {parent=auto_skirt_front_lf, loc=<0 0 0>, rot=<0 0 0 1>}
 dyn_skirt_front_01_lf {parent=offset_skirt_front_lf, loc=<0 0 0>, rot=<0.0344 -0.6364 0.0443 -0.7693>}
 dyn_skirt_front_02_lf {parent=dyn_skirt_front_01_lf, loc=<11.3931 0 0>, rot=<-0.016 -0.0855 0.1837 -0.9791>}
 dyn_skirt_front_03_lf {parent=dyn_skirt_front_02_lf, loc=<12.1589 0 0>, rot=<0 0 0 1>}
 base_skirt_front_side_lf {parent=pelvis, loc=<8.5165 3.8709 -8.083>, rot=<-0.2606 -0.6574 -0.2368 -0.6663>}
 auto_skirt_front_side_lf {parent=base_skirt_front_side_lf, loc=<0 0 0>, rot=<0 0 0 1>}
 offset_skirt_front_side_lf {parent=auto_skirt_front_side_lf, loc=<0 0 0>, rot=<0 0 0 1>}
 dyn_skirt_front_side_01_lf {parent=offset_skirt_front_side_lf, loc=<0 0 0>, rot=<0.0686 -0.5497 0.0863 -0.8281>}
 dyn_skirt_front_side_02_lf {parent=dyn_skirt_front_side_01_lf, loc=<11.7264 0 0>, rot=<0 0 0 1>}
 dyn_skirt_front_side_03_lf {parent=dyn_skirt_front_side_02_lf, loc=<9.1821 0 0>, rot=<0 0 0 1>}
 base_skirt_back_side_lf {parent=pelvis, loc=<9.2208 -4.6048 -8.4777>, rot=<0.2466 -0.6627 0.2702 -0.6534>}
 auto_skirt_back_side_lf {parent=base_skirt_back_side_lf, loc=<0 0 0>, rot=<0 0 0 1>}
 offset_skirt_back_side_lf {parent=auto_skirt_back_side_lf, loc=<0 0 0>, rot=<0 0 0 1>}
 dyn_skirt_back_side_01_lf {parent=offset_skirt_back_side_lf, loc=<0 0 0>, rot=<0.0574 -0.5035 -0.1497 -0.849>}
 dyn_skirt_back_side_02_lf {parent=dyn_skirt_back_side_01_lf, loc=<8.0819 0 0>, rot=<0 -0.0458 0 -0.999>}
 dyn_skirt_back_side_03_lf {parent=dyn_skirt_back_side_02_lf, loc=<9.1821 0 0>, rot=<0 0 0 1>}
 base_skirt_back_lf {parent=pelvis, loc=<9.2017 -7.1927 -4.7113>, rot=<0.491 -0.5089 0.5089 -0.491>}
 auto_skirt_back_lf {parent=base_skirt_back_lf, loc=<0 0 0>, rot=<0 0 0 1>}
 offset_skirt_back_lf {parent=auto_skirt_back_lf, loc=<0 0 0>, rot=<0 0 0 1>}
 dyn_skirt_back_01_lf {parent=offset_skirt_back_lf, loc=<0 0 0>, rot=<0.0913 -0.5193 -0.3205 -0.787>}
 dyn_skirt_back_02_lf {parent=dyn_skirt_back_01_lf, loc=<8.8514 0 0>, rot=<0 0 0 1>}
 dyn_skirt_back_03_lf {parent=dyn_skirt_back_02_lf, loc=<8.8514 0 0>, rot=<0 0 0 1>}
 base_loin_cloth_back_lf {parent=pelvis, loc=<8.8098 -7.6252 -5.246>, rot=<0.491 -0.5089 0.5089 -0.491>}
 auto_loin_cloth_back_lf {parent=base_loin_cloth_back_lf, loc=<0 0 0>, rot=<0 0 0 1>}
 offset_loin_cloth_back_lf {parent=auto_loin_cloth_back_lf, loc=<0 0 0>, rot=<0 0 0 1>}
 dyn_loin_cloth_back_01_lf {parent=offset_loin_cloth_back_lf, loc=<0 0 0>, rot=<0 -0.4672 0 -0.8841>}
 dyn_loin_cloth_back_02_lf {parent=dyn_loin_cloth_back_01_lf, loc=<7.6748 0 0>, rot=<0 -0.1472 0 -0.9891>}
 dyn_loin_cloth_back_03_lf {parent=dyn_loin_cloth_back_02_lf, loc=<6.8498 0 0>, rot=<0 0 0 1>}
 base_skirt_front_side_rt {parent=pelvis, loc=<8.517 3.8709 8.083>, rot=<0.6491 0.2806 0.6386 0.3036>}
 auto_skirt_front_side_rt {parent=base_skirt_front_side_rt, loc=<0 0 0>, rot=<0 0 0 1>}
 offset_skirt_front_side_rt {parent=auto_skirt_front_side_rt, loc=<0 0 0>, rot=<0 0 0 1>}
 dyn_skirt_front_side_01_rt {parent=offset_skirt_front_side_rt, loc=<0 0 0>, rot=<-0.0319 -0.553 -0.1411 -0.8205>}
 dyn_skirt_front_side_02_rt {parent=dyn_skirt_front_side_01_rt, loc=<11.7269 -0.0001 -0.0002>, rot=<0 0 0 -1>}
 dyn_skirt_front_side_03_rt {parent=dyn_skirt_front_side_02_rt, loc=<9.1821 0 0>, rot=<0 0 0 1>}
 base_skirt_front_rt {parent=pelvis, loc=<8.0005 7.1032 3.4928>, rot=<-0.5089 -0.491 -0.491 -0.5089>}
 auto_skirt_front_rt {parent=base_skirt_front_rt, loc=<0 0 0>, rot=<0 0 0 1>}
 offset_skirt_front_rt {parent=auto_skirt_front_rt, loc=<0 0 0>, rot=<0 0 0 1>}
 dyn_skirt_front_01_rt {parent=offset_skirt_front_rt, loc=<0 0 0>, rot=<-0.0344 -0.6364 -0.0443 -0.7693>}
 dyn_skirt_front_02_rt {parent=dyn_skirt_front_01_rt, loc=<11.3936 -0.0001 -0.0001>, rot=<0.016 -0.0855 -0.1837 -0.9791>}
 dyn_skirt_front_03_rt {parent=dyn_skirt_front_02_rt, loc=<12.1589 0 0>, rot=<0 0 0 1>}
 base_skirt_back_side_rt {parent=pelvis, loc=<9.2208 -4.6048 8.4777>, rot=<0.643 -0.2942 0.6531 -0.271>}
 auto_skirt_back_side_rt {parent=base_skirt_back_side_rt, loc=<0 0 0>, rot=<0 0 0 1>}
 offset_skirt_back_side_rt {parent=auto_skirt_back_side_rt, loc=<0 0 0>, rot=<0 0 0 1>}
 dyn_skirt_back_side_01_rt {parent=offset_skirt_back_side_rt, loc=<0 0 0>, rot=<-0.076 -0.5011 0.1811 -0.8428>}
 dyn_skirt_back_side_02_rt {parent=dyn_skirt_back_side_01_rt, loc=<8.0821 0 -0.0001>, rot=<0 -0.0458 0 -0.999>}
 dyn_skirt_back_side_03_rt {parent=dyn_skirt_back_side_02_rt, loc=<9.1819 0 0.0001>, rot=<0 0 0 1>}
 base_loin_cloth_back_rt {parent=pelvis, loc=<8.8098 -7.6252 5.246>, rot=<0.491 -0.5089 0.5089 -0.491>}
 auto_loin_cloth_back_rt {parent=base_loin_cloth_back_rt, loc=<0 0 0>, rot=<0 0 0 1>}
 offset_loin_cloth_back_rt {parent=auto_loin_cloth_back_rt, loc=<0 0 0>, rot=<0 0 0 1>}
 dyn_loin_cloth_back_01_rt {parent=offset_loin_cloth_back_rt, loc=<0 0 0>, rot=<0 -0.4672 0 -0.8841>}
 dyn_loin_cloth_back_02_rt {parent=dyn_loin_cloth_back_01_rt, loc=<7.6748 0 0>, rot=<0 -0.1472 0 -0.9891>}
 dyn_loin_cloth_back_03_rt {parent=dyn_loin_cloth_back_02_rt, loc=<6.8498 0 0>, rot=<0 0 0 1>}
 base_skirt_back_rt {parent=pelvis, loc=<9.2021 -7.1928 4.7113>, rot=<0.491 -0.5089 0.5089 -0.491>}
 auto_skirt_back_rt {parent=base_skirt_back_rt, loc=<0 0 0>, rot=<0 0 0 1>}
 offset_skirt_back_rt {parent=auto_skirt_back_rt, loc=<0 0 0>, rot=<0 0 0 1>}
 dyn_skirt_back_01_rt {parent=offset_skirt_back_rt, loc=<0 0 0>, rot=<-0.0913 -0.5193 0.3205 -0.787>}
 dyn_skirt_back_02_rt {parent=dyn_skirt_back_01_rt, loc=<8.8518 0.0001 -0.0002>, rot=<0 0 0 1>}
 dyn_skirt_back_03_rt {parent=dyn_skirt_back_02_rt, loc=<8.8514 0 0>, rot=<0 0 0 1>}
 base_bags_lf {parent=pelvis, loc=<11.1719 2.9222 -10.3397>, rot=<-0.0872 -0.9267 -0.2788 -0.2365>}
 offset_bags_lf {parent=base_bags_lf, loc=<0 0 0>, rot=<0 0 0 1>}
 dyn_bags_lf {parent=offset_bags_lf, loc=<0 0 0>, rot=<0 0 0 1>}
 base_bags_02_rt {parent=pelvis, loc=<10.1769 -2.6017 10.738>, rot=<0.2654 -0.1327 0.9546 -0.0277>}
 offset_bags_02_rt {parent=base_bags_02_rt, loc=<0 0 0>, rot=<0 0 0 1>}
 dyn_bags_02_rt {parent=offset_bags_02_rt, loc=<0 0 0>, rot=<0 0 0 1>}
 base_bags_01_rt {parent=pelvis, loc=<9.458 3.2498 10.5452>, rot=<0.2695 0.1846 0.9425 0.0703>}
 offset_bags_01_rt {parent=base_bags_01_rt, loc=<0 0 0>, rot=<0 0 0 1>}
 dyn_bags_01_rt {parent=offset_bags_01_rt, loc=<0 0 0>, rot=<0 0 0 1>}
 deform_glute_base_l {parent=pelvis, loc=<-7.3466 -9.0213 -9.1245>, rot=<0.0172 0.031 -0.4477 -0.8935>}
 deform_glute_l {parent=deform_glute_base_l, loc=<0 0 0>, rot=<0 0 0 1>}
 deform_glute_base_r {parent=pelvis, loc=<-7.3464 -9.0213 9.1245>, rot=<0.031 -0.0172 0.8935 -0.4477>}
 deform_glute_r {parent=deform_glute_base_r, loc=<0 0 0>, rot=<0 0 0 1>}
 base_granade_02_lf {parent=pelvis, loc=<12.6061 -4.9851 -9.1303>, rot=<0.0616 -0.9357 0.2791 -0.2067>}
 offset_granade_02_lf {parent=base_granade_02_lf, loc=<0 0 0>, rot=<0 0 0 1>}
 dyn_granade_02_lf {parent=offset_granade_02_lf, loc=<0 0 0>, rot=<0 0 0 1>}
 base_granade_01_lf {parent=pelvis, loc=<12.1397 -1.6191 -10.7549>, rot=<-0.0003 -0.9812 0.1068 -0.1609>}
 offset_granade_01_lf {parent=base_granade_01_lf, loc=<0 0 0>, rot=<0 0 0 1>}
 dyn_granade_01_lf {parent=offset_granade_01_lf, loc=<0 0 0>, rot=<0 0 0 1>}
 base_skirt_side_lf {parent=pelvis, loc=<9.0416 -0.6693 -9.8465>, rot=<-0.0127 -0.707 0.0127 -0.707>}
 auto_skirt_side_lf {parent=base_skirt_side_lf, loc=<0 0 0>, rot=<0.0144 -0.0015 -0.1041 -0.9945>}
 offset_skirt_side_lf {parent=auto_skirt_side_lf, loc=<0 0 0>, rot=<0 0 0 1>}
 dyn_skirt_side_01_lf {parent=offset_skirt_side_lf, loc=<0 0 0>, rot=<0.0067 -0.5317 0.02 -0.8466>}
 dyn_skirt_side_02_lf {parent=dyn_skirt_side_01_lf, loc=<9.1821 0 0>, rot=<0 0 0 1>}
 dyn_skirt_side_03_lf {parent=dyn_skirt_side_02_lf, loc=<9.1821 0 0>, rot=<0 0 0 1>}
 base_skirt_side_rt {parent=pelvis, loc=<9.0417 -0.6693 9.8464>, rot=<0.707 -0.0127 0.707 0.0127>}
 auto_skirt_side_rt {parent=base_skirt_side_rt, loc=<0 0 0>, rot=<0 0 0 1>}
 offset_skirt_side_rt {parent=auto_skirt_side_rt, loc=<0 0 0>, rot=<0 0 0 1>}
 dyn_skirt_side_01_rt {parent=offset_skirt_side_rt, loc=<0 0 0>, rot=<-0.0067 -0.5317 -0.02 -0.8466>}
 dyn_skirt_side_02_rt {parent=dyn_skirt_side_01_rt, loc=<9.1822 0 0>, rot=<0 0 0 1>}
 dyn_skirt_side_03_rt {parent=dyn_skirt_side_02_rt, loc=<9.182 0 0>, rot=<0 0 0 1>}
 aim_bags_lf {parent=pelvis, loc=<-8.6856 7.3126 -15.8077>, rot=<-0.0872 -0.9267 -0.2788 -0.2365>}
 aim_bags_02_rt {parent=pelvis, loc=<-2.6954 -3.9635 16.4707>, rot=<0.2654 -0.1327 0.9546 -0.0277>}
 aim_bags_01_rt {parent=pelvis, loc=<-3.9262 5.6453 15.1216>, rot=<0.2695 0.1846 0.9425 0.0703>}
 base_loin_cloth_front_cr {parent=pelvis, loc=<7.7328 8.7679 0>, rot=<-0.5089 -0.491 -0.491 -0.5089>}
 offset_loin_cloth_front_cr {parent=base_loin_cloth_front_cr, loc=<0 0 0>, rot=<0 0 0 1>}
 dyn_loin_cloth_front_01_cr {parent=offset_loin_cloth_front_cr, loc=<0 0 0>, rot=<0 -0.6319 0 -0.7751>}
 dyn_loin_cloth_front_02_cr {parent=dyn_loin_cloth_front_01_cr, loc=<7.4296 0 0>, rot=<0 -0.0546 0 -0.9985>}
 dyn_loin_cloth_front_03_cr {parent=dyn_loin_cloth_front_02_cr, loc=<7.4296 0 0>, rot=<0 0 0 1>}
 aim_belt_02 {parent=pelvis, loc=<-23.046 9.3062 12.6285>, rot=<-0.0127 -0.707 0.0127 -0.707>}
 aim_belt_01 {parent=pelvis, loc=<-18.4894 6.8724 18.052>, rot=<-0.0127 -0.707 0.0127 -0.707>}
 surface_belt_01_01 {parent=pelvis, loc=<-8.3342 6.1838 14.1529>, rot=<0.1807 0.2789 0.9395 -0.0827>}
 surface_belt_01_02 {parent=surface_belt_01_01, loc=<4.671 2.5174 -0.0763>, rot=<-0.044 0.0531 0.0045 -0.9976>}
 base_aimer_belt_01 {parent=surface_belt_01_02, loc=<4.7954 -0.5763 0.3467>, rot=<-0.4649 0.2745 0.4651 -0.7016>}
 aimer_belt_01 {parent=base_aimer_belt_01, loc=<0 0 0>, rot=<0 0 0 1>}
 base_belt_01_01 {parent=aimer_belt_01, loc=<1.4822 0.2155 -1.2795>, rot=<0.5294 -0.3551 -0.3781 -0.6713>}
 dyn_belt_01_01 {parent=base_belt_01_01, loc=<0 0 0>, rot=<0 0 0 1>}
 dyn_belt_01_02 {parent=dyn_belt_01_01, loc=<3.2193 0 0>, rot=<0 0 0 1>}
 dyn_belt_01_03 {parent=dyn_belt_01_02, loc=<3.1236 0 0>, rot=<0 0 0 1>}
 surface_belt_02_01 {parent=pelvis, loc=<-14.145 6.345 16.054>, rot=<0.1134 0.5677 0.8154 0.0016>}
 surface_belt_02_02 {parent=surface_belt_02_01, loc=<1.9311 -2.9599 0.4439>, rot=<0 0 0 1>}
 base_aimer_belt_02 {parent=surface_belt_02_02, loc=<4.518 1.5498 -0.0174>, rot=<-0.5875 0.3145 0.4211 -0.6153>}
 aimer_belt_02 {parent=base_aimer_belt_02, loc=<0 0 0>, rot=<0 0 0 1>}
 base_belt_02_01 {parent=aimer_belt_02, loc=<1.0737 -0.2798 -2.2858>, rot=<0.6561 -0.3406 -0.3581 -0.5704>}
 dyn_belt_02_01 {parent=base_belt_02_01, loc=<0 0 0>, rot=<0 0 0 1>}
 dyn_belt_02_02 {parent=dyn_belt_02_01, loc=<3.273 0 0>, rot=<0 0 0 1>}
 dyn_belt_02_03 {parent=dyn_belt_02_02, loc=<5.0131 0 0>, rot=<0 0 0 1>}
 aim_granade_01_lf {parent=pelvis, loc=<-3.4743 -1.421 -16.9066>, rot=<-0.001 -0.9835 0.081 -0.1617>}
 aim_granade_02_lf {parent=pelvis, loc=<-2.9988 -8.6426 -14.6158>, rot=<0.0613 -0.9369 0.2749 -0.2072>}
 thigh_l_ctrl {controlType=EULER_TRANSFORM, type=Control, loc=<0 0 0>, rot=<0 0 0 1>}
 calf_l_ctrl {controlType=EULER_TRANSFORM, parent=thigh_l_ctrl, type=Control, loc=<0 0 0>, rot=<0 0 0 1>}
 thigh_r_ctrl {controlType=EULER_TRANSFORM, type=Control, loc=<0 0 0>, rot=<0 0 0 1>}
 calf_r_ctrl {controlType=EULER_TRANSFORM, parent=thigh_r_ctrl, type=Control, loc=<0 0 0>, rot=<0 0 0 1>}
]
variables [crouch {type=bool, default=False}]
