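# Physics asset for the F_MED_ManicRufus female character body, covering the dangling front loin cloth and skirt bones plus the pelvis and thighs.
# Capsule bodies on the pelvis and thighs provide leg and hip shapes around the cloth chains.
# The loin cloth hangs from a freely rotating offset bone through three tightly limited links.
# Skirt chains rotate freely from the pelvis, except the left side's extra, tightly limited third segment, which the right lacks.
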PhysicsAsset /BRCosmetics/Characters/Player/Female/Medium/Bodies/F_MED_ManicRufus/Meshes/F_MED_ManicRufus_Physics
bodies {
 dyn_loin_cloth_front_01_cr [{shapeType=Capsule, length=13.0171, radius=1.2198, shapeName=dyn_loin_cloth_front_01_cr_capsule, center=<5 0 0>, rotation=<rot 90 0 90>}]
 dyn_loin_cloth_front_02_cr [{
   shapeType=Capsule
   length=13.6378
   radius=4.478
   shapeName=dyn_loin_cloth_front_02_cr_capsule
   center=<4.3867 0 3.4728>
   rotation=<rot 90 180 -90>
  }]
 dyn_loin_cloth_front_03_cr [{
   shapeType=Capsule
   length=17.9408
   radius=4.2896
   shapeName=dyn_loin_cloth_front_03_cr_capsule
   center=<2.7325 0 3.3604>
   rotation=<rot 90 180 -90>
  }]
 dyn_skirt_front_01_lf [{shapeType=Capsule, length=10, radius=2, shapeName=dyn_skirt_front_01_lf_capsule, center=<5 0 -0.7988>, rotation=<rot 90 0 0>}]
 dyn_skirt_front_01_rt [{shapeType=Capsule, length=10, radius=2, shapeName=dyn_skirt_front_01_rt_capsule, center=<5 0 -0.7988>, rotation=<rot 90 0 0>}]
 dyn_skirt_front_02_lf [{shapeType=Capsule, length=10, radius=2, shapeName=dyn_skirt_front_02_lf_capsule, center=<5 0 -1.6627>, rotation=<rot 90 0 0>}]
 dyn_skirt_front_02_rt [{shapeType=Capsule, length=10, radius=2, shapeName=dyn_skirt_front_02_rt_capsule, center=<5 0 -1.6627>, rotation=<rot 90 0 0>}]
 dyn_skirt_front_03_lf [{
   shapeType=Capsule
   length=4.6631
   radius=7.4967
   shapeName=dyn_skirt_front_02_lf_capsule
   center=<4.371 -4.7413 3.7565>
   rotation=<rot 69.9999 -105.0001 -0.0002>
  }]
 offset_loin_cloth_front_cr [{shapeType=Capsule, length=0.505, radius=0.505, shapeName=offset_loin_cloth_front_cr_capsule, rotation=<rot 0 -90 0>}]
 pelvis [{shapeType=Capsule, length=5.2668, radius=11.4035, rotation=<rot 0 -90 0>}]
 thigh_l [{shapeType=Capsule, length=26.8524, radius=7.8485, center=<-27.9282 -0.3321 -0.0977>, rotation=<rot 0.322 -81.4724 85.6793>}]
 thigh_r [{shapeType=Capsule, length=26.5756, radius=8.1253, center=<27.9282 0.3321 0.0977>, rotation=<rot 0.322 -81.4724 -94.3207>}]
}
constraints [
 {bone1Name=thigh_r, bone2Name=pelvis, swing1LimitDegrees=45, swing2LimitDegrees=45, twistLimitDegrees=45}
 {
  bone1Name=dyn_loin_cloth_front_03_cr
  bone2Name=dyn_loin_cloth_front_02_cr
  swing1LimitDegrees=0
  swing2LimitDegrees=10
  twistLimitDegrees=20
 }
 {
  bone1Name=offset_loin_cloth_front_cr
  bone2Name=pelvis
  swing1LimitDegrees=45
  swing1Motion=Free
  swing2LimitDegrees=45
  swing2Motion=Free
  twistLimitDegrees=45
  twistMotion=Free
 }
 {bone1Name=dyn_loin_cloth_front_01_cr, bone2Name=offset_loin_cloth_front_cr, swing1LimitDegrees=5, swing2LimitDegrees=5, twistLimitDegrees=0}
 {
  bone1Name=dyn_loin_cloth_front_02_cr
  bone2Name=dyn_loin_cloth_front_01_cr
  swing1LimitDegrees=0
  swing2LimitDegrees=10
  twistLimitDegrees=20
 }
 {bone1Name=thigh_l, bone2Name=pelvis, swing1LimitDegrees=45, swing2LimitDegrees=45, twistLimitDegrees=45}
 {
  bone1Name=dyn_skirt_front_01_lf
  bone2Name=pelvis
  swing1LimitDegrees=45
  swing1Motion=Free
  swing2LimitDegrees=45
  swing2Motion=Free
  twistLimitDegrees=45
  twistMotion=Free
 }
 {
  bone1Name=dyn_skirt_front_02_lf
  bone2Name=dyn_skirt_front_01_lf
  swing1LimitDegrees=45
  swing1Motion=Free
  swing2LimitDegrees=45
  swing2Motion=Free
  twistLimitDegrees=45
  twistMotion=Free
 }
 {
  bone1Name=dyn_skirt_front_01_rt
  bone2Name=pelvis
  swing1LimitDegrees=45
  swing1Motion=Free
  swing2LimitDegrees=45
  swing2Motion=Free
  twistLimitDegrees=45
  twistMotion=Free
 }
 {
  bone1Name=dyn_skirt_front_02_rt
  bone2Name=dyn_skirt_front_01_rt
  swing1LimitDegrees=45
  swing1Motion=Free
  swing2LimitDegrees=45
  swing2Motion=Free
  twistLimitDegrees=45
  twistMotion=Free
 }
 {bone1Name=dyn_skirt_front_03_lf, bone2Name=dyn_skirt_front_02_lf, swing1LimitDegrees=5, swing2LimitDegrees=20, twistLimitDegrees=0}
]
