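# Physics asset for the skirt of the F_MED_ManicRufus character body, covering right-side front and front-side skirt bone chains.
# Each chain has three capsule bodies hanging from the pelvis, which, with both thighs, provides capsule shapes for the body.
# The upper two joints of each chain rotate freely, while the last joint is tightly limited, with twist set to zero degrees.
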PhysicsAsset /BRCosmetics/Characters/Player/Female/Medium/Bodies/F_MED_ManicRufus/Meshes/F_MED_ManicRufus_Skirt_Physics
bodies {
 dyn_skirt_front_01_rt [{shapeType=Capsule, length=10, radius=2, shapeName=dyn_skirt_front_01_rt_capsule, center=<5 0 -0.7988>, rotation=<rot 90 0 0>}]
 dyn_skirt_front_02_rt [{shapeType=Capsule, length=10, radius=2, shapeName=dyn_skirt_front_02_rt_capsule, center=<5 0 -1.2973>, rotation=<rot 90 0 0>}]
 dyn_skirt_front_03_rt [{
   shapeType=Capsule
   length=5.5921
   radius=5.2146
   shapeName=dyn_skirt_front_02_rt_capsule
   center=<5.1726 0.9356 4.1201>
   rotation=<rot 80 0.0001 0.0001>
  }]
 dyn_skirt_front_side_01_rt [{shapeType=Capsule, length=4.251, radius=3.8307, shapeName=dyn_skirt_front_side_01_rt_capsule, center=<5 0 0>, rotation=<rot 90 0 0>}]
 dyn_skirt_front_side_02_rt [{shapeType=Capsule, length=6.0719, radius=3.6199, shapeName=dyn_skirt_front_side_02_rt_capsule, center=<5 0 0>, rotation=<rot 90 0 0>}]
 dyn_skirt_front_side_03_rt [{
   shapeType=Capsule
   length=6.3875
   radius=5.3885
   shapeName=dyn_skirt_front_side_03_rt_capsule
   center=<5 0 1.277>
   rotation=<rot 90 180 -109.9998>
  }]
 pelvis [{shapeType=Capsule, length=3.3852, radius=13.2851, rotation=<rot 0 -90 0>}]
 thigh_l [{shapeType=Capsule, length=26.5756, radius=8.1253, center=<-27.9282 -0.3321 -0.0977>, rotation=<rot 0.322 -81.4724 85.6793>}]
 thigh_r [{shapeType=Capsule, length=26.5756, radius=8.1253, center=<27.9282 0.3321 0.0977>, rotation=<rot 0.322 -81.4724 -94.3207>}]
}
constraints [
 {bone1Name=thigh_r, bone2Name=pelvis, swing1LimitDegrees=45, swing2LimitDegrees=45, twistLimitDegrees=45}
 {bone1Name=thigh_l, bone2Name=pelvis, swing1LimitDegrees=45, swing2LimitDegrees=45, twistLimitDegrees=45}
 {
  bone1Name=dyn_skirt_front_01_rt
  bone2Name=pelvis
  swing1LimitDegrees=45
  swing1Motion=Free
  swing2LimitDegrees=45
  swing2Motion=Free
  twistLimitDegrees=45
  twistMotion=Free
 }
 {
  bone1Name=dyn_skirt_front_02_rt
  bone2Name=dyn_skirt_front_01_rt
  swing1LimitDegrees=45
  swing1Motion=Free
  swing2LimitDegrees=45
  swing2Motion=Free
  twistLimitDegrees=45
  twistMotion=Free
 }
 {bone1Name=dyn_skirt_front_03_rt, bone2Name=dyn_skirt_front_02_rt, swing1LimitDegrees=5, swing2LimitDegrees=10, twistLimitDegrees=0}
 {
  bone1Name=dyn_skirt_front_side_01_rt
  bone2Name=pelvis
  swing1LimitDegrees=45
  swing1Motion=Free
  swing2LimitDegrees=45
  swing2Motion=Free
  twistLimitDegrees=45
  twistMotion=Free
 }
 {
  bone1Name=dyn_skirt_front_side_02_rt
  bone2Name=dyn_skirt_front_side_01_rt
  swing1LimitDegrees=45
  swing1Motion=Free
  swing2LimitDegrees=45
  swing2Motion=Free
  twistLimitDegrees=45
  twistMotion=Free
 }
 {bone1Name=dyn_skirt_front_side_03_rt, bone2Name=dyn_skirt_front_side_02_rt, swing1LimitDegrees=0, swing2LimitDegrees=20, twistLimitDegrees=0}
]
