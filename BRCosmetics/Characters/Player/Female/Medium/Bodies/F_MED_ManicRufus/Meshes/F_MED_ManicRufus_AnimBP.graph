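# Post-process style Anim Blueprint for the ManicRufus medium female cosmetic body, adding secondary motion to skirt, loincloth, belts and pouches.
# It copies the parent mesh pose through a crouch-aware Control Rig and bone offsets, then cached chains of spring-driven AnimDynamics.
# Later, RigidBody simulations relative to thigh and pelvis bracket weighted constraints that make belt bones follow right-side skirt bones.
# Additive bone tweaks correct the result, some active only when `bRigidBodyEnabled` is false.
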AnimBlueprint /BRCosmetics/Characters/Player/Female/Medium/Bodies/F_MED_ManicRufus/Meshes/F_MED_ManicRufus_AnimBP
target_skeleton=&/F_MED_ManicRufus_Skeleton
graphs [
 AnimGraph {nodes [
   n1 = RigidBody!(ComponentPose=Local To Component(LocalPose=Use cached pose 'slide_cache'()), Alpha, OverrideWorldGravity:Vector="0,0,0", ExternalForce) -> (Pose) {
    bClampLinearTranslationLimitToRefPose=true
    bOverrideWorldGravity=true
    overridePhysicsAsset=&/F_MED_ManicRufus_Skirt_Physics
    simulationSpace=BaseBoneSpace
    alphaBoolBlend {blendOption=Linear}
    baseBoneRef {boneName=thigh_r}
    componentAppliedLinearAccClamp=<6000 6000 6000>
    simSpaceSettings {worldAlpha=0.1, externalLinearVelocity {z=10}}
   }
   n2 = "Constraint - surface_belt_02_01 (PPPPPPPP)"!(
    ComponentPose=@n1
    Alpha=1
    ConstraintWeights_0=0.4551
    ConstraintWeights_1:float=0.3039
    ConstraintWeights_2:float=0.1256
    ConstraintWeights_3:float=0.0633
    ConstraintWeights_4:float=0.0307
    ConstraintWeights_5:float=0.0125
    ConstraintWeights_6:float=0.0047
    ConstraintWeights_7:float=0.0042
   ) -> (Pose) {
    alphaBoolBlend {blendOption=Linear}
    boneToModify {boneName=surface_belt_02_01}
    constraintSetup [
     {targetBone {boneName=dyn_skirt_front_03_rt}}
     {targetBone {boneName=dyn_skirt_front_side_03_rt}}
     {targetBone {boneName=dyn_skirt_front_side_02_rt}}
     {targetBone {boneName=dyn_skirt_side_03_rt}}
     {targetBone {boneName=dyn_skirt_front_02_rt}}
     {targetBone {boneName=dyn_skirt_side_02_rt}}
     {targetBone {boneName=dyn_skirt_back_side_03_rt}}
     {targetBone {boneName=dyn_skirt_back_side_02_rt}}
    ]
    constraintWeights [1,1,1,1,1,1,1,1]
   }
   n3 = "Constraint - surface_belt_02_02 (PPPPPPPP)"!(
    ComponentPose=@n2
    Alpha=1
    ConstraintWeights_0=0.7778
    ConstraintWeights_1:float=0.1148
    ConstraintWeights_2:float=0.0581
    ConstraintWeights_3:float=0.0326
    ConstraintWeights_4:float=0.0091
    ConstraintWeights_5:float=0.0046
    ConstraintWeights_6:float=0.002
    ConstraintWeights_7:float=0.0011
   ) -> (Pose) {
    alphaBoolBlend {blendOption=Linear}
    boneToModify {boneName=surface_belt_02_02}
    constraintSetup [
     {targetBone {boneName=dyn_skirt_front_03_rt}}
     {targetBone {boneName=dyn_skirt_front_side_03_rt}}
     {targetBone {boneName=dyn_skirt_front_side_02_rt}}
     {targetBone {boneName=dyn_skirt_front_02_rt}}
     {targetBone {boneName=dyn_skirt_side_03_rt}}
     {targetBone {boneName=dyn_skirt_front_side_01_rt}}
     {targetBone {boneName=dyn_skirt_front_01_rt}}
     {targetBone {boneName=dyn_skirt_side_02_rt}}
    ]
    constraintWeights [1,1,1,1,1,1,1,1]
   }
   n4 = "Constraint - base_aimer_belt_02 (PPPPPPPP)"!(
    ComponentPose=@n3
    Alpha=1
    ConstraintWeights_0=0.8307
    ConstraintWeights_1:float=0.1417
    ConstraintWeights_2:float=0.0151
    ConstraintWeights_3:float=0.0087
    ConstraintWeights_4:float=0.0028
    ConstraintWeights_5:float=0.0004
    ConstraintWeights_6:float=0.0002
    ConstraintWeights_7:float=0.0002
   ) -> (Pose) {
    alphaBoolBlend {blendOption=Linear}
    boneToModify {boneName=base_aimer_belt_02}
    constraintSetup [
     {targetBone {boneName=dyn_skirt_front_03_rt}}
     {targetBone {boneName=dyn_skirt_front_side_03_rt}}
     {targetBone {boneName=dyn_skirt_front_side_02_rt}}
     {targetBone {boneName=dyn_skirt_side_03_rt}}
     {targetBone {boneName=dyn_skirt_front_02_rt}}
     {targetBone {boneName=dyn_skirt_back_side_03_rt}}
     {targetBone {boneName=dyn_skirt_front_side_01_rt}}
     {targetBone {boneName=dyn_skirt_front_01_rt}}
    ]
    constraintWeights [1,1,1,1,1,1,1,1]
   }
   n5 = "Constraint - base_belt_02_01 (PPPPPPPP)"!(
    ComponentPose=@n4
    Alpha=1
    ConstraintWeights_0=0.8307
    ConstraintWeights_1:float=0.1417
    ConstraintWeights_2:float=0.0151
    ConstraintWeights_3:float=0.0087
    ConstraintWeights_4:float=0.0028
    ConstraintWeights_5:float=0.0004
    ConstraintWeights_6:float=0.0002
    ConstraintWeights_7:float=0.0002
   ) -> (Pose) {
    alphaBoolBlend {blendOption=Linear}
    boneToModify {boneName=base_belt_02_01}
    constraintSetup [
     {targetBone {boneName=dyn_skirt_front_03_rt}}
     {targetBone {boneName=dyn_skirt_front_side_03_rt}}
     {targetBone {boneName=dyn_skirt_front_side_02_rt}}
     {targetBone {boneName=dyn_skirt_side_03_rt}}
     {targetBone {boneName=dyn_skirt_front_02_rt}}
     {targetBone {boneName=dyn_skirt_back_side_03_rt}}
     {targetBone {boneName=dyn_skirt_front_side_01_rt}}
     {targetBone {boneName=dyn_skirt_front_01_rt}}
    ]
    constraintWeights [1,1,1,1,1,1,1,1]
   }
   n6 = "Constraint - surface_belt_01_01 (PPPPPPPP)"!(
    ComponentPose=@n5
    Alpha=1
    ConstraintWeights_0=0.3283
    ConstraintWeights_1:float=0.2527
    ConstraintWeights_2:float=0.1452
    ConstraintWeights_3:float=0.0994
    ConstraintWeights_4:float=0.0974
    ConstraintWeights_5:float=0.0325
    ConstraintWeights_6:float=0.0324
    ConstraintWeights_7:float=0.0121
   ) -> (Pose) {
    alphaBoolBlend {blendOption=Linear}
    boneToModify {boneName=surface_belt_01_01}
    constraintSetup [
     {targetBone {boneName=dyn_skirt_front_03_rt}}
     {targetBone {boneName=dyn_skirt_front_side_02_rt}}
     {targetBone {boneName=dyn_skirt_front_02_rt}}
     {targetBone {boneName=dyn_skirt_front_side_03_rt}}
     {targetBone {boneName=dyn_skirt_front_side_01_rt}}
     {targetBone {boneName=dyn_skirt_front_01_rt}}
     {targetBone {boneName=dyn_skirt_side_02_rt}}
     {targetBone {boneName=dyn_skirt_back_side_02_rt}}
    ]
    constraintWeights [1,1,1,1,1,1,1,1]
   }
   n7 = "Constraint - surface_belt_01_02 (PPPPPPPP)"!(
    ComponentPose=@n6
    Alpha=1
    ConstraintWeights_0=0.355
    ConstraintWeights_1:float=0.2079
    ConstraintWeights_2:float=0.1866
    ConstraintWeights_3:float=0.1463
    ConstraintWeights_4:float=0.0456
    ConstraintWeights_5:float=0.0243
    ConstraintWeights_6:float=0.0184
    ConstraintWeights_7:float=0.0159
   ) -> (Pose) {
    alphaBoolBlend {blendOption=Linear}
    boneToModify {boneName=surface_belt_01_02}
    constraintSetup [
     {targetBone {boneName=dyn_skirt_front_side_03_rt}}
     {targetBone {boneName=dyn_skirt_front_03_rt}}
     {targetBone {boneName=dyn_skirt_side_03_rt}}
     {targetBone {boneName=dyn_skirt_front_side_02_rt}}
     {targetBone {boneName=dyn_skirt_side_02_rt}}
     {targetBone {boneName=dyn_skirt_back_side_03_rt}}
     {targetBone {boneName=dyn_skirt_back_side_02_rt}}
     {targetBone {boneName=dyn_skirt_front_02_rt}}
    ]
    constraintWeights [1,1,1,1,1,1,1,1]
   }
   n8 = "Constraint - base_aimer_belt_01 (PPPPPPPP)"!(
    ComponentPose=@n7
    Alpha=1
    ConstraintWeights_0=0.4845
    ConstraintWeights_1:float=0.2401
    ConstraintWeights_2:float=0.1845
    ConstraintWeights_3:float=0.056
    ConstraintWeights_4:float=0.0172
    ConstraintWeights_5:float=0.0127
    ConstraintWeights_6:float=0.0038
    ConstraintWeights_7:float=0.0013
   ) -> (Pose) {
    alphaBoolBlend {blendOption=Linear}
    boneToModify {boneName=base_aimer_belt_01}
    constraintSetup [
     {targetBone {boneName=dyn_skirt_front_side_03_rt}}
     {targetBone {boneName=dyn_skirt_front_03_rt}}
     {targetBone {boneName=dyn_skirt_side_03_rt}}
     {targetBone {boneName=dyn_skirt_front_side_02_rt}}
     {targetBone {boneName=dyn_skirt_back_side_03_rt}}
     {targetBone {boneName=dyn_skirt_side_02_rt}}
     {targetBone {boneName=dyn_skirt_back_side_02_rt}}
     {targetBone {boneName=dyn_skirt_front_02_rt}}
    ]
    constraintWeights [1,1,1,1,1,1,1,1]
   }
   n9 = "Constraint - base_belt_01_01 (PPPPPPPP)"!(
    ComponentPose=@n8
    Alpha=1
    ConstraintWeights_0=0.4845
    ConstraintWeights_1:float=0.2401
    ConstraintWeights_2:float=0.1845
    ConstraintWeights_3:float=0.056
    ConstraintWeights_4:float=0.0172
    ConstraintWeights_5:float=0.0127
    ConstraintWeights_6:float=0.0038
    ConstraintWeights_7:float=0.0013
   ) -> (Pose) {
    alphaBoolBlend {blendOption=Linear}
    boneToModify {boneName=base_belt_01_01}
    constraintSetup [
     {targetBone {boneName=dyn_skirt_front_side_03_rt}}
     {targetBone {boneName=dyn_skirt_front_03_rt}}
     {targetBone {boneName=dyn_skirt_side_03_rt}}
     {targetBone {boneName=dyn_skirt_front_side_02_rt}}
     {targetBone {boneName=dyn_skirt_back_side_03_rt}}
     {targetBone {boneName=dyn_skirt_side_02_rt}}
     {targetBone {boneName=dyn_skirt_back_side_02_rt}}
     {targetBone {boneName=dyn_skirt_front_02_rt}}
    ]
    constraintWeights [1,1,1,1,1,1,1,1]
   }
   n10 = "Transform (Modify) Bone - Bone: surface_belt_02_01"(ComponentPose=@n9, Alpha=1, Translation="0,0,0", Rotation="0,0,0", Scale="1,1,1") {
    translationMode=BMM_Additive
    translationSpace=BCS_BoneSpace
    alphaBoolBlend {blendOption=Linear}
    boneToModify {boneName=surface_belt_02_01}
   }
   n11 = "Transform (Modify) Bone - Bone: surface_belt_01_01"(ComponentPose=@n10, Alpha=1, Translation="0,0,-1", Rotation="0,0,0", Scale="1,1,1") {
    translationMode=BMM_Additive
    translationSpace=BCS_BoneSpace
    alphaBoolBlend {blendOption=Linear}
    boneToModify {boneName=surface_belt_01_01}
    translation {z=-1}
   }
   n12 = "Transform (Modify) Bone - Bone: base_belt_01_01"(ComponentPose=@n11, Alpha=1, Translation="0,1,1", Rotation="0,0,-20", Scale="1,1,1") {
    rotationMode=BMM_Additive
    rotationSpace=BCS_BoneSpace
    translationMode=BMM_Additive
    translationSpace=BCS_BoneSpace
    alphaBoolBlend {blendOption=Linear}
    boneToModify {boneName=base_belt_01_01}
    rotation {roll=-20}
    translation {y=1, z=1}
   }
   n13 = "Transform (Modify) Bone - Bone: base_belt_02_01"(ComponentPose=@n12, Alpha=1, Translation="0,1,1", Rotation="0,0,-20", Scale="1,1,1") {
    rotationMode=BMM_Additive
    rotationSpace=BCS_BoneSpace
    translationMode=BMM_Additive
    translationSpace=BCS_BoneSpace
    alphaBoolBlend {blendOption=Linear}
    boneToModify {boneName=base_belt_02_01}
    rotation {roll=-20}
    translation {y=1, z=1}
   }
   n14 = RigidBody(ComponentPose=@n13, OverrideWorldGravity="0,0,0") {
    bClampLinearTranslationLimitToRefPose=true
    bOverrideWorldGravity=true
    overridePhysicsAsset=&/F_MED_ManicRufus_Physics
    simulationSpace=BaseBoneSpace
    alphaBoolBlend {blendOption=Linear}
    baseBoneRef {boneName=pelvis}
    componentAppliedLinearAccClamp=<6000 6000 6000>
    simSpaceSettings {worldAlpha=0.05, externalLinearVelocity {z=10}}
   }
   n15 = RigidBody(ComponentPose=@n14, OverrideWorldGravity="0,0,0") {
    bClampLinearTranslationLimitToRefPose=true
    bOverrideWorldGravity=true
    overridePhysicsAsset=&/F_MED_ManicRufus_Belt_Physics
    simulationSpace=BaseBoneSpace
    alphaBoolBlend {blendOption=Linear}
    baseBoneRef {boneName=thigh_r}
    componentAppliedLinearAccClamp=<6000 6000 6000>
    simSpaceSettings {worldAlpha=0.5, externalLinearVelocity {z=10}}
   }
   n16 = "Transform (Modify) Bone - Bone: dyn_skirt_front_02_lf"!(ComponentPose=@n15, bAlphaBoolEnabled:bool=True, Translation="0,0,0", Rotation="20,0,0", Scale="1,1,1") -> (Pose) {
    alphaInputType=Bool
    rotationMode=BMM_Additive
    rotationSpace=BCS_BoneSpace
    alphaBoolBlend {blendInTime=0.2, blendOption=Linear, blendOutTime=0.2}
    boneToModify {boneName=dyn_skirt_front_02_lf}
    rotation {pitch=20}
   }
   n17 = "Transform (Modify) Bone - Bone: base_belt_02_01"!(ComponentPose=@n16, bAlphaBoolEnabled:bool=True, Translation="0,2.5,-0.5", Rotation="0,0,10", Scale="1,1,1") -> (Pose) {
    alphaInputType=Bool
    rotationMode=BMM_Additive
    rotationSpace=BCS_BoneSpace
    scaleMode=BMM_Additive
    translationMode=BMM_Additive
    translationSpace=BCS_BoneSpace
    alphaBoolBlend {blendInTime=0.2, blendOption=Linear, blendOutTime=0.2}
    boneToModify {boneName=base_belt_02_01}
    rotation {roll=10}
    translation {y=2.5, z=-0.5}
   }
   n18 = "Transform (Modify) Bone - Bone: base_belt_01_01"!(ComponentPose=@n17, bAlphaBoolEnabled:bool=True, Translation="0,0,-2", Rotation="0,10,0", Scale="1,1,1") -> (Pose) {
    alphaInputType=Bool
    rotationMode=BMM_Additive
    rotationSpace=BCS_BoneSpace
    translationMode=BMM_Additive
    translationSpace=BCS_BoneSpace
    alphaBoolBlend {blendInTime=0.2, blendOption=Linear, blendOutTime=0.2}
    boneToModify {boneName=base_belt_01_01}
    rotation {yaw=10}
    translation {z=-2}
   }
   n19 = "Transform (Modify) Bone - Bone: dyn_skirt_front_02_lf"(ComponentPose=@n18, bAlphaBoolEnabled=True, Translation="0,0,0", Rotation="10,0.203,-0.317", Scale="1,1,1") {
    alphaInputType=Bool
    rotationMode=BMM_Additive
    rotationSpace=BCS_BoneSpace
    alphaBoolBlend {blendOption=Linear}
    boneToModify {boneName=dyn_skirt_front_02_lf}
    rotation=<rot 10 0.203 -0.317>
   }
   n20 = "Transform (Modify) Bone - Bone: dyn_skirt_front_02_rt"!(ComponentPose=@n19, bAlphaBoolEnabled:bool=True, Translation="0,0,0", Rotation="5,0,0", Scale="1,1,1") -> (Pose) {
    alphaInputType=Bool
    rotationMode=BMM_Additive
    rotationSpace=BCS_BoneSpace
    alphaBoolBlend {blendOption=Linear}
    boneToModify {boneName=dyn_skirt_front_02_rt}
    rotation {pitch=5}
   }
   n21 = "Transform (Modify) Bone - Bone: dyn_skirt_front_side_02_rt"!(ComponentPose=@n20, bAlphaBoolEnabled:bool=True, Translation="0,0,0", Rotation="0,0,0", Scale="1,1,1") -> (Pose) {
    alphaInputType=Bool
    rotationMode=BMM_Additive
    rotationSpace=BCS_BoneSpace
    alphaBoolBlend {blendOption=Linear}
    boneToModify {boneName=dyn_skirt_front_side_02_rt}
   }
   n22 = NOT Boolean(A=Get bRigidBodyEnabled())
   n23 = "Transform (Modify) Bone - Bone: offset_skirt_front_side_rt"!(ComponentPose=@n21, bAlphaBoolEnabled:bool=@n22, Translation="0,0,0", Rotation="3,0,0", Scale="1,1,1") -> (Pose) {
    alphaInputType=Bool
    rotationMode=BMM_Additive
    rotationSpace=BCS_BoneSpace
    alphaBoolBlend {blendOption=Linear}
    boneToModify {boneName=offset_skirt_front_side_rt}
    rotation {pitch=3}
   }
   n24 = "Transform (Modify) Bone - Bone: offset_skirt_front_rt"!(ComponentPose=@n23, bAlphaBoolEnabled:bool=@n22, Translation="0,0,0", Rotation="2,0,0", Scale="1,1,1") -> (Pose) {
    alphaInputType=Bool
    rotationMode=BMM_Additive
    rotationSpace=BCS_BoneSpace
    alphaBoolBlend {blendOption=Linear}
    boneToModify {boneName=offset_skirt_front_rt}
    rotation {pitch=2}
   }
   n25 = "Transform (Modify) Bone - Bone: dyn_loin_cloth_front_03_cr"!(ComponentPose=@n24, bAlphaBoolEnabled:bool=True, Translation="0,0,0", Rotation="5,0,0", Scale="1,1,1") -> (Pose) {
    alphaInputType=Bool
    rotationMode=BMM_Additive
    rotationSpace=BCS_BoneSpace
    alphaBoolBlend {blendOption=Linear}
    boneToModify {boneName=dyn_loin_cloth_front_03_cr}
    rotation {pitch=5}
   }
   n26 = "Transform (Modify) Bone - Bone: base_belt_01_01"(ComponentPose=@n25, bAlphaBoolEnabled=True, Translation="0,-2,0", Rotation="0,0,-20", Scale="1,1,1") {
    alphaInputType=Bool
    rotationMode=BMM_Additive
    rotationSpace=BCS_BoneSpace
    translationMode=BMM_Additive
    translationSpace=BCS_BoneSpace
    alphaBoolBlend {blendOption=Linear}
    boneToModify {boneName=base_belt_01_01}
    rotation {roll=-20}
    translation {y=-2}
   }
   n27 = "Transform (Modify) Bone - Bone: base_belt_02_01"(ComponentPose=@n26, bAlphaBoolEnabled=True, Translation="0,0,-1", Rotation="0,0,0", Scale="1,1,1") {
    alphaInputType=Bool
    rotationMode=BMM_Additive
    rotationSpace=BCS_BoneSpace
    translationMode=BMM_Additive
    translationSpace=BCS_BoneSpace
    alphaBoolBlend {blendOption=Linear}
    boneToModify {boneName=base_belt_02_01}
    translation {z=-1}
   }
   Output Pose(Result=Component To Local(ComponentPose=@n27)) {layerGroup=None}
   n28 = Control Rig!(Source=Copy Pose From Mesh() {bUseAttachedParent=true}, Alpha, crouch:bool=Get bIsCrouching()) -> (Pose) {
    alphaBoolBlend {blendOption=Linear}
    controlRigAssetReference {blueprintRigClass=&/F_MED_ManicRufus_CtrlRig.F_MED_ManicRufus_CtrlRig_C}
   }
   n29 = "Transform (Modify) Bone - Bone: dyn_granade_02_lf"!(ComponentPose=Local To Component(LocalPose=@n28), bAlphaBoolEnabled:bool=True, Translation="1,-1.5,-1", Rotation="0,-10,0", Scale="1,1,1") -> (Pose) {
    alphaInputType=Bool
    rotationMode=BMM_Additive
    rotationSpace=BCS_BoneSpace
    translationMode=BMM_Additive
    translationSpace=BCS_BoneSpace
    alphaBoolBlend {blendOption=Linear}
    boneToModify {boneName=dyn_granade_02_lf}
    rotation {yaw=-10}
    translation=<1 -1.5 -1>
   }
   n30 = "Transform (Modify) Bone - Bone: base_granade_01_lf"!(ComponentPose=@n29, bAlphaBoolEnabled:bool=True, Translation="1,-1,0.03", Rotation="0,-30,0", Scale="1,1,1") -> (Pose) {
    alphaInputType=Bool
    rotationMode=BMM_Additive
    rotationSpace=BCS_BoneSpace
    translationMode=BMM_Additive
    translationSpace=BCS_BoneSpace
    alphaBoolBlend {blendOption=Linear}
    boneToModify {boneName=base_granade_01_lf}
    rotation {yaw=-30}
    translation=<1 -1 0.03>
   }
   n31 = "Transform (Modify) Bone - Bone: dyn_bags_lf"!(ComponentPose=@n30, bAlphaBoolEnabled:bool=True, Translation="1.256,2.176,0.480", Rotation="0,-30,0", Scale="1,1,1") -> (Pose) {
    alphaInputType=Bool
    rotationMode=BMM_Additive
    rotationSpace=BCS_BoneSpace
    translationMode=BMM_Additive
    translationSpace=BCS_BoneSpace
    alphaBoolBlend {blendOption=Linear}
    boneToModify {boneName=dyn_bags_lf}
    rotation {yaw=-30}
    translation=<1.256 2.176 0.48>
   }
   n32 = "Transform (Modify) Bone - Bone: dyn_bags_02_rt"!(ComponentPose=@n31, bAlphaBoolEnabled:bool=True, Translation="-0.5,0.5,0", Rotation="0,0,0", Scale="1,1,1") -> (Pose) {
    alphaInputType=Bool
    rotationMode=BMM_Additive
    rotationSpace=BCS_BoneSpace
    translationMode=BMM_Additive
    translationSpace=BCS_BoneSpace
    alphaBoolBlend {blendOption=Linear}
    boneToModify {boneName=dyn_bags_02_rt}
    translation=<-0.5 0.5>
   }
   n33 = "Transform (Modify) Bone - Bone: dyn_bags_01_rt"!(ComponentPose=@n32, bAlphaBoolEnabled:bool=True, Translation="2.616,-0.292,-0.099", Rotation="0,20,0", Scale="1,1,1") -> (Pose) {
    alphaInputType=Bool
    rotationMode=BMM_Additive
    rotationSpace=BCS_BoneSpace
    translationMode=BMM_Additive
    translationSpace=BCS_BoneSpace
    alphaBoolBlend {blendOption=Linear}
    boneToModify {boneName=dyn_bags_01_rt}
    rotation {yaw=20}
    translation=<2.616 -0.292 -0.099>
   }
   Save cached pose 'ControlRigCache'(Pose=Component To Local(ComponentPose=@n33)) {cacheName=ControlRigCache}
   n34 = Anim Dynamics - Bone: dyn_skirt_front_03_lf!(
    ComponentPose=Local To Component(LocalPose=Use cached pose 'ControlRigCache'())
    Alpha=1
    GravityOverride:Vector="0,0,-980"
    ExternalForce="0,0,0"
   ) -> (Pose) {
    angularDampingOverride=1
    angularSpringConstant=20000
    bAngularSpring=true
    bOverrideAngularDamping=true
    bUseGravityOverride=true
    gravityScale=0.35
    simulationSpace=RootRelative
    alphaBoolBlend {blendOption=Linear}
    boundBone {boneName=dyn_skirt_front_03_lf}
    componentAppliedLinearAccClamp=<6000 6000 6000>
    componentLinearAccScale=<1 1 1>
    componentLinearVelScale=<1 1 1>
    gravityOverride {z=-980}
    physicsBodyDefinitions [{
      boundBone {boneName=dyn_skirt_front_03_lf}
      boxExtents=<30 30 30>
      constraintSetup {angularLimitsMax=<0 0 0>, angularLimitsMin=<0 -20 0>, angularTarget=<1 0 0>}
      localJointOffset=<4 0 0>
     }]
   }
   n35 = Anim Dynamics - Bone: dyn_skirt_front_side_02_lf!(ComponentPose=@n34, Alpha=1, GravityOverride:Vector="0,0,-980", ExternalForce="0,0,0") -> (Pose) {
    angularDampingOverride=1
    angularSpringConstant=20000
    bAngularSpring=true
    bOverrideAngularDamping=true
    bUseGravityOverride=true
    gravityScale=0.35
    simulationSpace=RootRelative
    alphaBoolBlend {blendOption=Linear}
    boundBone {boneName=dyn_skirt_front_side_02_lf}
    componentAppliedLinearAccClamp=<6000 6000 6000>
    componentLinearAccScale=<1 1 1>
    componentLinearVelScale=<1 1 1>
    gravityOverride {z=-980}
    physicsBodyDefinitions [{
      boundBone {boneName=dyn_skirt_front_side_02_lf}
      boxExtents=<30 30 30>
      constraintSetup {angularLimitsMax=<0 0 0>, angularLimitsMin=<0 -15 0>, angularTarget=<1 0 0>}
      localJointOffset=<4 0 0>
     }]
   }
   n36 = Anim Dynamics - Bone: dyn_skirt_front_side_03_lf!(ComponentPose=@n35, Alpha=1, GravityOverride:Vector="0,0,-980", ExternalForce="0,0,0") -> (Pose) {
    angularDampingOverride=1
    angularSpringConstant=20000
    bAngularSpring=true
    bOverrideAngularDamping=true
    bUseGravityOverride=true
    gravityScale=0.35
    simulationSpace=RootRelative
    alphaBoolBlend {blendOption=Linear}
    boundBone {boneName=dyn_skirt_front_side_03_lf}
    componentAppliedLinearAccClamp=<6000 6000 6000>
    componentLinearAccScale=<1 1 1>
    componentLinearVelScale=<1 1 1>
    gravityOverride {z=-980}
    physicsBodyDefinitions [{
      boundBone {boneName=dyn_skirt_front_side_03_lf}
      boxExtents=<30 30 30>
      constraintSetup {angularLimitsMax=<0 0 0>, angularLimitsMin=<0 -20 0>, angularTarget=<1 0 0>}
      localJointOffset=<4 0 0>
     }]
   }
   n37 = Anim Dynamics - Bone: dyn_skirt_side_02_lf!(ComponentPose=@n36, Alpha=1, GravityOverride:Vector="0,0,-980", ExternalForce="0,0,0") -> (Pose) {
    angularDampingOverride=1
    angularSpringConstant=20000
    bAngularSpring=true
    bOverrideAngularDamping=true
    bUseGravityOverride=true
    gravityScale=0.35
    simulationSpace=RootRelative
    alphaBoolBlend {blendOption=Linear}
    boundBone {boneName=dyn_skirt_side_02_lf}
    componentAppliedLinearAccClamp=<6000 6000 6000>
    componentLinearAccScale=<1 1 1>
    componentLinearVelScale=<1 1 1>
    gravityOverride {z=-980}
    physicsBodyDefinitions [{
      boundBone {boneName=dyn_skirt_side_02_lf}
      boxExtents=<30 30 30>
      constraintSetup {angularLimitsMax=<0 0 0>, angularLimitsMin=<0 -15 0>, angularTarget=<1 0 0>}
      localJointOffset=<4 0 0>
     }]
   }
   n38 = Anim Dynamics - Bone: dyn_skirt_side_03_lf!(ComponentPose=@n37, Alpha=1, GravityOverride:Vector="0,0,-980", ExternalForce="0,0,0") -> (Pose) {
    angularDampingOverride=1
    angularSpringConstant=20000
    bAngularSpring=true
    bOverrideAngularDamping=true
    bUseGravityOverride=true
    gravityScale=0.35
    simulationSpace=RootRelative
    alphaBoolBlend {blendOption=Linear}
    boundBone {boneName=dyn_skirt_side_03_lf}
    componentAppliedLinearAccClamp=<6000 6000 6000>
    componentLinearAccScale=<1 1 1>
    componentLinearVelScale=<1 1 1>
    gravityOverride {z=-980}
    physicsBodyDefinitions [{
      boundBone {boneName=dyn_skirt_side_03_lf}
      boxExtents=<30 30 30>
      constraintSetup {angularLimitsMax=<0 0 0>, angularLimitsMin=<0 -20 0>, angularTarget=<1 0 0>}
      localJointOffset=<4 0 0>
     }]
   }
   n39 = Anim Dynamics - Bone: dyn_skirt_back_side_02_lf!(ComponentPose=@n38, Alpha=1, GravityOverride:Vector="0,0,-980", ExternalForce="0,0,0") -> (Pose) {
    angularDampingOverride=1
    angularSpringConstant=20000
    bAngularSpring=true
    bOverrideAngularDamping=true
    bUseGravityOverride=true
    gravityScale=0.35
    simulationSpace=RootRelative
    alphaBoolBlend {blendOption=Linear}
    boundBone {boneName=dyn_skirt_back_side_02_lf}
    componentAppliedLinearAccClamp=<6000 6000 6000>
    componentLinearAccScale=<1 1 1>
    componentLinearVelScale=<1 1 1>
    gravityOverride {z=-980}
    physicsBodyDefinitions [{
      boundBone {boneName=dyn_skirt_back_side_02_lf}
      boxExtents=<30 30 30>
      constraintSetup {angularLimitsMax=<0 0 0>, angularLimitsMin=<0 -15 0>, angularTarget=<1 0 0>}
      localJointOffset=<4 0 0>
     }]
   }
   n40 = Anim Dynamics - Bone: dyn_skirt_back_side_03_lf!(ComponentPose=@n39, Alpha=1, GravityOverride:Vector="0,0,-980", ExternalForce="0,0,0") -> (Pose) {
    angularDampingOverride=1
    angularSpringConstant=20000
    bAngularSpring=true
    bOverrideAngularDamping=true
    bUseGravityOverride=true
    gravityScale=0.35
    simulationSpace=RootRelative
    alphaBoolBlend {blendOption=Linear}
    boundBone {boneName=dyn_skirt_back_side_03_lf}
    componentAppliedLinearAccClamp=<6000 6000 6000>
    componentLinearAccScale=<1 1 1>
    componentLinearVelScale=<1 1 1>
    gravityOverride {z=-980}
    physicsBodyDefinitions [{
      boundBone {boneName=dyn_skirt_back_side_03_lf}
      boxExtents=<30 30 30>
      constraintSetup {angularLimitsMax=<0 0 0>, angularLimitsMin=<0 -20 0>, angularTarget=<1 0 0>}
      localJointOffset=<4 0 0>
     }]
   }
   n41 = Anim Dynamics - Bone: dyn_skirt_back_03_lf!(ComponentPose=@n40, Alpha=1, GravityOverride:Vector="0,0,-980", ExternalForce="0,0,0") -> (Pose) {
    angularDampingOverride=1
    angularSpringConstant=20000
    bAngularSpring=true
    bOverrideAngularDamping=true
    bUseGravityOverride=true
    gravityScale=0.35
    simulationSpace=RootRelative
    alphaBoolBlend {blendOption=Linear}
    boundBone {boneName=dyn_skirt_back_03_lf}
    componentAppliedLinearAccClamp=<6000 6000 6000>
    componentLinearAccScale=<1 1 1>
    componentLinearVelScale=<1 1 1>
    gravityOverride {z=-980}
    physicsBodyDefinitions [{
      boundBone {boneName=dyn_skirt_back_03_lf}
      boxExtents=<30 30 30>
      constraintSetup {angularLimitsMax=<0 0 0>, angularLimitsMin=<0 -20 0>, angularTarget=<1 0 0>}
      localJointOffset=<4 0 0>
     }]
   }
   n42 = Anim Dynamics - Bone: dyn_loin_cloth_back_01_lf!(ComponentPose=@n41, Alpha=1, GravityOverride:Vector="0,0,-980", ExternalForce="0,0,0") -> (Pose) {
    angularDampingOverride=1
    angularSpringConstant=20000
    bAngularSpring=true
    bOverrideAngularDamping=true
    bUseGravityOverride=true
    gravityScale=0.35
    simulationSpace=RootRelative
    alphaBoolBlend {blendOption=Linear}
    boundBone {boneName=dyn_loin_cloth_back_01_lf}
    componentAppliedLinearAccClamp=<6000 6000 6000>
    componentLinearAccScale=<1 1 1>
    componentLinearVelScale=<1 1 1>
    gravityOverride {z=-980}
    physicsBodyDefinitions [{
      boundBone {boneName=dyn_loin_cloth_back_01_lf}
      boxExtents=<50 50 50>
      constraintSetup {angularLimitsMax=<0 0 0>, angularLimitsMin=<0 -25 0>, angularTarget=<1 0 0>}
      localJointOffset=<5 0 5>
     }]
   }
   n43 = Anim Dynamics - Bone: dyn_loin_cloth_back_02_lf!(ComponentPose=@n42, Alpha=1, GravityOverride:Vector="0,0,-980", ExternalForce="0,0,0") -> (Pose) {
    angularDampingOverride=1
    angularSpringConstant=20000
    bAngularSpring=true
    bOverrideAngularDamping=true
    bUseGravityOverride=true
    gravityScale=0.35
    simulationSpace=RootRelative
    alphaBoolBlend {blendOption=Linear}
    boundBone {boneName=dyn_loin_cloth_back_02_lf}
    componentAppliedLinearAccClamp=<6000 6000 6000>
    componentLinearAccScale=<1 1 1>
    componentLinearVelScale=<1 1 1>
    gravityOverride {z=-980}
    physicsBodyDefinitions [{
      boundBone {boneName=dyn_loin_cloth_back_02_lf}
      boxExtents=<50 50 50>
      constraintSetup {angularLimitsMax=<0 0 0>, angularLimitsMin=<0 -30 0>, angularTarget=<1 0 0>}
      localJointOffset=<8 0 2>
     }]
   }
   n44 = Anim Dynamics - Bone: dyn_loin_cloth_back_03_lf!(ComponentPose=@n43, Alpha=1, GravityOverride:Vector="0,0,-980", ExternalForce="0,0,0") -> (Pose) {
    angularDampingOverride=1
    angularSpringConstant=20000
    bAngularSpring=true
    bOverrideAngularDamping=true
    bUseGravityOverride=true
    gravityScale=0.35
    simulationSpace=RootRelative
    alphaBoolBlend {blendOption=Linear}
    boundBone {boneName=dyn_loin_cloth_back_03_lf}
    componentAppliedLinearAccClamp=<6000 6000 6000>
    componentLinearAccScale=<1 1 1>
    componentLinearVelScale=<1 1 1>
    gravityOverride {z=-980}
    physicsBodyDefinitions [{
      boundBone {boneName=dyn_loin_cloth_back_03_lf}
      boxExtents=<50 50 50>
      constraintSetup {angularLimitsMax=<0 0 0>, angularLimitsMin=<0 -30 0>, angularTarget=<1 0 0>}
      localJointOffset=<5 0 5>
     }]
   }
   Save cached pose 'animdyn_Left_Cache'(Pose=Component To Local(ComponentPose=@n44)) {cacheName=animdyn_Left_Cache}
   n45 = Anim Dynamics - Bone: dyn_skirt_front_03_rt!(
    ComponentPose=Local To Component(LocalPose=Use cached pose 'animdyn_Left_Cache'())
    Alpha=1
    GravityOverride:Vector="0,0,-980"
    ExternalForce="0,0,0"
   ) -> (Pose) {
    angularDampingOverride=1
    angularSpringConstant=20000
    bAngularSpring=true
    bOverrideAngularDamping=true
    bUseGravityOverride=true
    gravityScale=0.35
    simulationSpace=RootRelative
    alphaBoolBlend {blendOption=Linear}
    boundBone {boneName=dyn_skirt_front_03_rt}
    componentAppliedLinearAccClamp=<6000 6000 6000>
    componentLinearAccScale=<1 1 1>
    componentLinearVelScale=<1 1 1>
    gravityOverride {z=-980}
    physicsBodyDefinitions [{
      boundBone {boneName=dyn_skirt_front_03_rt}
      boxExtents=<20 30 30>
      constraintSetup {angularLimitsMax=<0 0 0>, angularLimitsMin=<0 -20 0>, angularTarget=<1 0 0>}
      localJointOffset=<6 0 0>
     }]
   }
   n46 = Anim Dynamics - Bone: dyn_skirt_front_side_02_rt!(ComponentPose=@n45, Alpha=1, GravityOverride:Vector="0,0,-980", ExternalForce="0,0,0") -> (Pose) {
    angularDampingOverride=1
    angularSpringConstant=20000
    bAngularSpring=true
    bOverrideAngularDamping=true
    bUseGravityOverride=true
    gravityScale=0.35
    simulationSpace=RootRelative
    alphaBoolBlend {blendOption=Linear}
    boundBone {boneName=dyn_skirt_front_side_02_rt}
    componentAppliedLinearAccClamp=<6000 6000 6000>
    componentLinearAccScale=<1 1 1>
    componentLinearVelScale=<1 1 1>
    gravityOverride {z=-980}
    physicsBodyDefinitions [{
      boundBone {boneName=dyn_skirt_front_side_02_rt}
      boxExtents=<20 30 30>
      constraintSetup {angularLimitsMax=<0 0 0>, angularLimitsMin=<0 -15 0>, angularTarget=<1 0 0>}
      localJointOffset=<4 0 0>
     }]
   }
   n47 = Anim Dynamics - Bone: dyn_skirt_front_side_03_rt!(ComponentPose=@n46, Alpha=1, GravityOverride:Vector="0,0,-980", ExternalForce="0,0,0") -> (Pose) {
    angularDampingOverride=1
    angularSpringConstant=20000
    bAngularSpring=true
    bOverrideAngularDamping=true
    bUseGravityOverride=true
    gravityScale=0.35
    simulationSpace=RootRelative
    alphaBoolBlend {blendOption=Linear}
    boundBone {boneName=dyn_skirt_front_side_03_rt}
    componentAppliedLinearAccClamp=<6000 6000 6000>
    componentLinearAccScale=<1 1 1>
    componentLinearVelScale=<1 1 1>
    gravityOverride {z=-980}
    physicsBodyDefinitions [{
      boundBone {boneName=dyn_skirt_front_side_03_rt}
      boxExtents=<20 30 30>
      constraintSetup {angularLimitsMax=<0 0 0>, angularLimitsMin=<0 -20 0>, angularTarget=<1 0 0>}
      localJointOffset=<6 0 0>
     }]
   }
   n48 = Anim Dynamics - Bone: dyn_skirt_side_02_rt!(ComponentPose=@n47, Alpha=1, GravityOverride:Vector="0,0,-980", ExternalForce="0,0,0") -> (Pose) {
    angularDampingOverride=1
    angularSpringConstant=20000
    bAngularSpring=true
    bOverrideAngularDamping=true
    bUseGravityOverride=true
    gravityScale=0.35
    simulationSpace=RootRelative
    alphaBoolBlend {blendOption=Linear}
    boundBone {boneName=dyn_skirt_side_02_rt}
    componentAppliedLinearAccClamp=<6000 6000 6000>
    componentLinearAccScale=<1 1 1>
    componentLinearVelScale=<1 1 1>
    gravityOverride {z=-980}
    physicsBodyDefinitions [{
      boundBone {boneName=dyn_skirt_side_02_rt}
      boxExtents=<20 30 30>
      constraintSetup {angularLimitsMax=<0 0 0>, angularLimitsMin=<0 -15 0>, angularTarget=<1 0 0>}
      localJointOffset=<4 0 0>
     }]
   }
   n49 = Anim Dynamics - Bone: dyn_skirt_side_03_rt!(ComponentPose=@n48, Alpha=1, GravityOverride:Vector="0,0,-980", ExternalForce="0,0,0") -> (Pose) {
    angularDampingOverride=1
    angularSpringConstant=20000
    bAngularSpring=true
    bOverrideAngularDamping=true
    bUseGravityOverride=true
    gravityScale=0.35
    simulationSpace=RootRelative
    alphaBoolBlend {blendOption=Linear}
    boundBone {boneName=dyn_skirt_side_03_rt}
    componentAppliedLinearAccClamp=<6000 6000 6000>
    componentLinearAccScale=<1 1 1>
    componentLinearVelScale=<1 1 1>
    gravityOverride {z=-980}
    physicsBodyDefinitions [{
      boundBone {boneName=dyn_skirt_side_03_rt}
      boxExtents=<20 30 30>
      constraintSetup {angularLimitsMax=<0 0 0>, angularLimitsMin=<0 -20 0>, angularTarget=<1 0 0>}
      localJointOffset=<6 0 0>
     }]
   }
   n50 = Anim Dynamics - Bone: dyn_skirt_back_side_02_rt!(ComponentPose=@n49, Alpha=1, GravityOverride:Vector="0,0,-980", ExternalForce="0,0,0") -> (Pose) {
    angularDampingOverride=1
    angularSpringConstant=20000
    bAngularSpring=true
    bOverrideAngularDamping=true
    bUseGravityOverride=true
    gravityScale=0.35
    simulationSpace=RootRelative
    alphaBoolBlend {blendOption=Linear}
    boundBone {boneName=dyn_skirt_back_side_02_rt}
    componentAppliedLinearAccClamp=<6000 6000 6000>
    componentLinearAccScale=<1 1 1>
    componentLinearVelScale=<1 1 1>
    gravityOverride {z=-980}
    physicsBodyDefinitions [{
      boundBone {boneName=dyn_skirt_back_side_02_rt}
      boxExtents=<20 30 30>
      constraintSetup {angularLimitsMax=<0 0 0>, angularLimitsMin=<0 -15 0>, angularTarget=<1 0 0>}
      localJointOffset=<4 0 0>
     }]
   }
   n51 = Anim Dynamics - Bone: dyn_skirt_back_side_03_rt!(ComponentPose=@n50, Alpha=1, GravityOverride:Vector="0,0,-980", ExternalForce="0,0,0") -> (Pose) {
    angularDampingOverride=1
    angularSpringConstant=20000
    bAngularSpring=true
    bOverrideAngularDamping=true
    bUseGravityOverride=true
    gravityScale=0.35
    simulationSpace=RootRelative
    alphaBoolBlend {blendOption=Linear}
    boundBone {boneName=dyn_skirt_back_side_03_rt}
    componentAppliedLinearAccClamp=<6000 6000 6000>
    componentLinearAccScale=<1 1 1>
    componentLinearVelScale=<1 1 1>
    gravityOverride {z=-980}
    physicsBodyDefinitions [{
      boundBone {boneName=dyn_skirt_back_side_03_rt}
      boxExtents=<20 30 30>
      constraintSetup {angularLimitsMax=<0 0 0>, angularLimitsMin=<0 -20 0>, angularTarget=<1 0 0>}
      localJointOffset=<6 0 0>
     }]
   }
   n52 = Anim Dynamics - Bone: dyn_skirt_back_03_rt!(ComponentPose=@n51, Alpha=1, GravityOverride:Vector="0,0,-980", ExternalForce="0,0,0") -> (Pose) {
    angularDampingOverride=1
    angularSpringConstant=20000
    bAngularSpring=true
    bOverrideAngularDamping=true
    bUseGravityOverride=true
    gravityScale=0.35
    simulationSpace=RootRelative
    alphaBoolBlend {blendOption=Linear}
    boundBone {boneName=dyn_skirt_back_03_rt}
    componentAppliedLinearAccClamp=<6000 6000 6000>
    componentLinearAccScale=<1 1 1>
    componentLinearVelScale=<1 1 1>
    gravityOverride {z=-980}
    physicsBodyDefinitions [{
      boundBone {boneName=dyn_skirt_back_03_rt}
      boxExtents=<20 30 30>
      constraintSetup {angularLimitsMax=<0 0 0>, angularLimitsMin=<0 -20 0>, angularTarget=<1 0 0>}
      localJointOffset=<6 0 0>
     }]
   }
   n53 = Anim Dynamics - Bone: dyn_loin_cloth_back_01_rt!(ComponentPose=@n52, Alpha=1, GravityOverride:Vector="0,0,-980", ExternalForce="0,0,0") -> (Pose) {
    angularDampingOverride=1
    angularSpringConstant=20000
    bAngularSpring=true
    bOverrideAngularDamping=true
    bUseGravityOverride=true
    gravityScale=0.35
    simulationSpace=RootRelative
    alphaBoolBlend {blendOption=Linear}
    boundBone {boneName=dyn_loin_cloth_back_01_rt}
    componentAppliedLinearAccClamp=<6000 6000 6000>
    componentLinearAccScale=<1 1 1>
    componentLinearVelScale=<1 1 1>
    gravityOverride {z=-980}
    physicsBodyDefinitions [{
      boundBone {boneName=dyn_loin_cloth_back_01_rt}
      boxExtents=<50 50 50>
      constraintSetup {angularLimitsMax=<0 0 0>, angularLimitsMin=<0 -25 0>, angularTarget=<1 0 0>}
      localJointOffset=<10 0 5>
     }]
   }
   n54 = Anim Dynamics - Bone: dyn_loin_cloth_back_02_rt!(ComponentPose=@n53, Alpha=1, GravityOverride:Vector="0,0,-980", ExternalForce="0,0,0") -> (Pose) {
    angularDampingOverride=1
    angularSpringConstant=20000
    bAngularSpring=true
    bOverrideAngularDamping=true
    bUseGravityOverride=true
    gravityScale=0.35
    simulationSpace=RootRelative
    alphaBoolBlend {blendOption=Linear}
    boundBone {boneName=dyn_loin_cloth_back_02_rt}
    componentAppliedLinearAccClamp=<6000 6000 6000>
    componentLinearAccScale=<1 1 1>
    componentLinearVelScale=<1 1 1>
    gravityOverride {z=-980}
    physicsBodyDefinitions [{
      boundBone {boneName=dyn_loin_cloth_back_02_rt}
      boxExtents=<50 50 50>
      constraintSetup {angularLimitsMax=<0 0 0>, angularLimitsMin=<0 -30 0>, angularTarget=<1 0 0>}
      localJointOffset=<10 0 5>
     }]
   }
   n55 = Anim Dynamics - Bone: dyn_loin_cloth_back_03_rt!(ComponentPose=@n54, Alpha=1, GravityOverride:Vector="0,0,-980", ExternalForce="0,0,0") -> (Pose) {
    angularDampingOverride=1
    angularSpringConstant=20000
    bAngularSpring=true
    bOverrideAngularDamping=true
    bUseGravityOverride=true
    gravityScale=0.35
    simulationSpace=RootRelative
    alphaBoolBlend {blendOption=Linear}
    boundBone {boneName=dyn_loin_cloth_back_03_rt}
    componentAppliedLinearAccClamp=<6000 6000 6000>
    componentLinearAccScale=<1 1 1>
    componentLinearVelScale=<1 1 1>
    gravityOverride {z=-980}
    physicsBodyDefinitions [{
      boundBone {boneName=dyn_loin_cloth_back_03_rt}
      boxExtents=<50 50 50>
      constraintSetup {angularLimitsMax=<0 0 0>, angularLimitsMin=<0 -30 0>, angularTarget=<1 0 0>}
      localJointOffset=<10 0 5>
     }]
   }
   Save cached pose 'animdyn_Right_Cache'(Pose=Component To Local(ComponentPose=@n55)) {cacheName=animdyn_Right_Cache}
   n56 = Anim Dynamics - Bone: dyn_arm_belt_02_rt!(
    ComponentPose=Local To Component(LocalPose=Use cached pose 'animdyn_Right_Cache'())
    Alpha=1
    GravityOverride:Vector="0,0,-980"
    ExternalForce="0,0,0"
   ) -> (Pose) {
    angularDampingOverride=1
    angularSpringConstant=20000
    bAngularSpring=true
    bOverrideAngularDamping=true
    bUseGravityOverride=true
    gravityScale=0.35
    simulationSpace=RootRelative
    alphaBoolBlend {blendOption=Linear}
    boundBone {boneName=dyn_arm_belt_02_rt}
    componentAppliedLinearAccClamp=<6000 6000 6000>
    componentLinearAccScale=<1 1 1>
    componentLinearVelScale=<1 1 1>
    gravityOverride {z=-980}
    physicsBodyDefinitions [{
      boundBone {boneName=dyn_arm_belt_02_rt}
      boxExtents=<20 30 30>
      constraintSetup {angularLimitsMax=<0 0 10>, angularLimitsMin=<0 0 -20>, angularTarget=<1 0 0>}
      localJointOffset=<8 0 0>
     }]
   }
   n57 = Anim Dynamics - Bone: dyn_arm_belt_01_rt!(ComponentPose=@n56, Alpha=1, GravityOverride:Vector="0,0,-980", ExternalForce="0,0,0") -> (Pose) {
    angularDampingOverride=1
    angularSpringConstant=20000
    bAngularSpring=true
    bOverrideAngularDamping=true
    bUseGravityOverride=true
    gravityScale=0.35
    simulationSpace=RootRelative
    alphaBoolBlend {blendOption=Linear}
    boundBone {boneName=dyn_arm_belt_01_rt}
    componentAppliedLinearAccClamp=<6000 6000 6000>
    componentLinearAccScale=<1 1 1>
    componentLinearVelScale=<1 1 1>
    gravityOverride {z=-980}
    physicsBodyDefinitions [{
      boundBone {boneName=dyn_arm_belt_01_rt}
      boxExtents=<40 30 30>
      constraintSetup {angularLimitsMax=<0 0 10>, angularLimitsMin=<0 0 -10>, angularTarget=<1 0 0>}
      localJointOffset=<4 0 0>
     }]
   }
   Save cached pose 'belt_skirt_surface_constraint'(Pose=Component To Local(ComponentPose=@n57)) {cacheName=belt_skirt_surface_constraint}
   n58 = Anim Dynamics - Bone: dyn_bags_02_rt(ComponentPose=Local To Component(LocalPose=Use cached pose 'belt_skirt_surface_constraint'()), Alpha=1, ExternalForce="0,0,0") {
    angularDampingOverride=1
    angularSpringConstant=20000
    bAngularSpring=true
    bOverrideAngularDamping=true
    bUseGravityOverride=true
    gravityScale=0.35
    simulationSpace=RootRelative
    alphaBoolBlend {blendOption=Linear}
    boundBone {boneName=dyn_bags_02_rt}
    componentAppliedLinearAccClamp=<6000 6000 6000>
    componentLinearAccScale=<1 1 1>
    componentLinearVelScale=<1 1 1>
    gravityOverride {z=-980}
    physicsBodyDefinitions [{
      boundBone {boneName=dyn_bags_02_rt}
      boxExtents=<40 40 40>
      constraintSetup {angularLimitsMax=<0 0 0>, angularLimitsMin=<0 -5 0>, angularTarget=<1 0 0>}
      localJointOffset=<2 0 0>
     }]
   }
   n59 = Anim Dynamics - Bone: dyn_bags_01_rt(ComponentPose=@n58, Alpha=1, ExternalForce="0,0,0") {
    angularDampingOverride=1
    angularSpringConstant=20000
    bAngularSpring=true
    bOverrideAngularDamping=true
    bUseGravityOverride=true
    gravityScale=0.35
    simulationSpace=RootRelative
    alphaBoolBlend {blendOption=Linear}
    boundBone {boneName=dyn_bags_01_rt}
    componentAppliedLinearAccClamp=<6000 6000 6000>
    componentLinearAccScale=<1 1 1>
    componentLinearVelScale=<1 1 1>
    gravityOverride {z=-980}
    physicsBodyDefinitions [{
      boundBone {boneName=dyn_bags_01_rt}
      boxExtents=<40 40 40>
      constraintSetup {angularLimitsMax=<0 0 0>, angularLimitsMin=<0 -5 0>, angularTarget=<1 0 0>}
      localJointOffset=<2 0 0>
     }]
   }
   n60 = Anim Dynamics - Bone: dyn_granade_01_lf(ComponentPose=@n59, Alpha=1, ExternalForce="0,0,0") {
    angularDampingOverride=1
    angularSpringConstant=20000
    bAngularSpring=true
    bOverrideAngularDamping=true
    bUseGravityOverride=true
    gravityScale=0.35
    simulationSpace=RootRelative
    alphaBoolBlend {blendOption=Linear}
    boundBone {boneName=dyn_granade_01_lf}
    componentAppliedLinearAccClamp=<6000 6000 6000>
    componentLinearAccScale=<1 1 1>
    componentLinearVelScale=<1 1 1>
    gravityOverride {z=-980}
    physicsBodyDefinitions [{
      boundBone {boneName=dyn_granade_01_lf}
      boxExtents=<40 40 40>
      constraintSetup {angularLimitsMax=<0 0 0>, angularLimitsMin=<0 -5 0>, angularTarget=<1 0 0>}
      localJointOffset=<2 0 0>
     }]
   }
   n61 = Anim Dynamics - Bone: dyn_granade_02_lf(ComponentPose=@n60, Alpha=1, ExternalForce="0,0,0") {
    angularDampingOverride=1
    angularSpringConstant=20000
    bAngularSpring=true
    bOverrideAngularDamping=true
    bUseGravityOverride=true
    gravityScale=0.35
    simulationSpace=RootRelative
    alphaBoolBlend {blendOption=Linear}
    boundBone {boneName=dyn_granade_02_lf}
    componentAppliedLinearAccClamp=<6000 6000 6000>
    componentLinearAccScale=<1 1 1>
    componentLinearVelScale=<1 1 1>
    gravityOverride {z=-980}
    physicsBodyDefinitions [{
      boundBone {boneName=dyn_granade_02_lf}
      boxExtents=<40 40 40>
      constraintSetup {angularLimitsMax=<0 0 0>, angularLimitsMin=<0 -5 0>, angularTarget=<1 0 0>}
      localJointOffset=<2 0 0>
     }]
   }
   n62 = Anim Dynamics - Bone: dyn_bags_lf(ComponentPose=@n61, Alpha=1, ExternalForce="0,0,0") {
    angularDampingOverride=1
    angularSpringConstant=20000
    bAngularSpring=true
    bOverrideAngularDamping=true
    bUseGravityOverride=true
    gravityScale=0.35
    simulationSpace=RootRelative
    alphaBoolBlend {blendOption=Linear}
    boundBone {boneName=dyn_bags_lf}
    componentAppliedLinearAccClamp=<6000 6000 6000>
    componentLinearAccScale=<1 1 1>
    componentLinearVelScale=<1 1 1>
    gravityOverride {z=-980}
    physicsBodyDefinitions [{
      boundBone {boneName=dyn_bags_lf}
      boxExtents=<40 40 40>
      constraintSetup {angularLimitsMax=<0 0 0>, angularLimitsMin=<0 -5 0>, angularTarget=<1 0 0>}
      localJointOffset=<2 0 0>
     }]
   }
   Save cached pose 'Bag_AnimDyn_Cache'(Pose=Component To Local(ComponentPose=@n62)) {cacheName=Bag_AnimDyn_Cache}
   n63 = "Transform (Modify) Bone - Bone: offset_skirt_side_lf"!(
    ComponentPose=Local To Component(LocalPose=Use cached pose 'Bag_AnimDyn_Cache'())
    bAlphaBoolEnabled:bool=True
    Translation="0,0,0"
    Rotation="20,0,0"
    Scale="1,1,1"
   ) -> (Pose) {
    alphaInputType=Bool
    rotationMode=BMM_Additive
    rotationSpace=BCS_BoneSpace
    alphaBoolBlend {blendOption=Linear}
    boneToModify {boneName=offset_skirt_side_lf}
    rotation {pitch=20}
   }
   n64 = "Transform (Modify) Bone - Bone: offset_skirt_front_side_lf"!(ComponentPose=@n63, bAlphaBoolEnabled:bool=True, Translation="0,0,0", Rotation="20,0,0", Scale="1,1,1") -> (Pose) {
    alphaInputType=Bool
    rotationMode=BMM_Additive
    rotationSpace=BCS_BoneSpace
    alphaBoolBlend {blendOption=Linear}
    boneToModify {boneName=offset_skirt_front_side_lf}
    rotation {pitch=20}
   }
   n65 = "Transform (Modify) Bone - Bone: offset_skirt_back_side_lf"!(ComponentPose=@n64, bAlphaBoolEnabled:bool=True, Translation="0,0,0", Rotation="20,0,0", Scale="1,1,1") -> (Pose) {
    alphaInputType=Bool
    rotationMode=BMM_Additive
    rotationSpace=BCS_BoneSpace
    alphaBoolBlend {blendOption=Linear}
    boneToModify {boneName=offset_skirt_back_side_lf}
    rotation {pitch=20}
   }
   n66 = "Transform (Modify) Bone - Bone: offset_skirt_back_lf"!(ComponentPose=@n65, bAlphaBoolEnabled:bool=True, Translation="0,0,0", Rotation="20,0,0", Scale="1,1,1") -> (Pose) {
    alphaInputType=Bool
    rotationMode=BMM_Additive
    rotationSpace=BCS_BoneSpace
    alphaBoolBlend {blendOption=Linear}
    boneToModify {boneName=offset_skirt_back_lf}
    rotation {pitch=20}
   }
   n67 = "Transform (Modify) Bone - Bone: offset_skirt_side_lf"(ComponentPose=@n66, bAlphaBoolEnabled=True, Translation="0,0,0", Rotation="20,0,0", Scale="1,1,1") {
    alphaInputType=Bool
    rotationMode=BMM_Additive
    rotationSpace=BCS_BoneSpace
    alphaBoolBlend {blendOption=Linear}
    boneToModify {boneName=offset_skirt_side_lf}
    rotation {pitch=20}
   }
   n68 = "Transform (Modify) Bone - Bone: offset_skirt_front_lf"!(ComponentPose=@n67, bAlphaBoolEnabled:bool=True, Translation="0,0,0", Rotation="5,0,0", Scale="1,1,1") -> (Pose) {
    alphaInputType=Bool
    rotationMode=BMM_Additive
    rotationSpace=BCS_BoneSpace
    alphaBoolBlend {blendOption=Linear}
    boneToModify {boneName=offset_skirt_front_lf}
    rotation {pitch=5}
   }
   n69 = "Transform (Modify) Bone - Bone: offset_skirt_front_side_lf"(ComponentPose=@n68, bAlphaBoolEnabled=True, Translation="0,0,0", Rotation="20,0,0", Scale="1,1,1") {
    alphaInputType=Bool
    rotationMode=BMM_Additive
    rotationSpace=BCS_BoneSpace
    alphaBoolBlend {blendOption=Linear}
    boneToModify {boneName=offset_skirt_front_side_lf}
    rotation {pitch=20}
   }
   n70 = "Transform (Modify) Bone - Bone: offset_skirt_front_lf"(ComponentPose=@n69, bAlphaBoolEnabled=True, Translation="0,0,0", Rotation="20,0,0", Scale="1,1,1") {
    alphaInputType=Bool
    rotationMode=BMM_Additive
    rotationSpace=BCS_BoneSpace
    alphaBoolBlend {blendOption=Linear}
    boneToModify {boneName=offset_skirt_front_lf}
    rotation {pitch=20}
   }
   n71 = "Transform (Modify) Bone - Bone: offset_loin_cloth_front_cr"!(ComponentPose=@n70, bAlphaBoolEnabled:bool=NOT Boolean(A=Get bIsCrouching()), Translation="0,0,0", Rotation="-12,-10,0", Scale="1,1,1") -> (Pose) {
    alphaInputType=Bool
    rotationMode=BMM_Additive
    rotationSpace=BCS_BoneSpace
    alphaBoolBlend {blendOption=Linear}
    boneToModify {boneName=offset_loin_cloth_front_cr}
    rotation {pitch=-12, yaw=-10}
   }
   n72 = "Transform (Modify) Bone - Bone: offset_loin_cloth_front_cr"(ComponentPose=@n71, bAlphaBoolEnabled=True, Translation="0,0,0", Rotation="10,0.051,0.136", Scale="1,1,1") {
    alphaInputType=Bool
    rotationMode=BMM_Additive
    rotationSpace=BCS_BoneSpace
    alphaBoolBlend {blendOption=Linear}
    boneToModify {boneName=offset_loin_cloth_front_cr}
    rotation=<rot 10 0.051 0.136>
   }
   n73 = "Transform (Modify) Bone - Bone: auto_skirt_front_side_lf"!(ComponentPose=@n72, bAlphaBoolEnabled:bool=True, Translation="0,0,0", Rotation="0,0,0", Scale="1,1,1") -> (Pose) {
    alphaInputType=Bool
    rotationMode=BMM_Additive
    rotationSpace=BCS_BoneSpace
    alphaBoolBlend {blendOption=Linear}
    boneToModify {boneName=auto_skirt_front_side_lf}
   }
   n74 = "Transform (Modify) Bone - Bone: auto_skirt_front_lf"!(ComponentPose=@n73, bAlphaBoolEnabled:bool=True, Translation="0,0,0", Rotation="10,60,0", Scale="1,1,1") -> (Pose) {
    alphaInputType=Bool
    rotationMode=BMM_Additive
    rotationSpace=BCS_BoneSpace
    alphaBoolBlend {blendOption=Linear}
    boneToModify {boneName=auto_skirt_front_lf}
    rotation {pitch=10, yaw=60}
   }
   n75 = "Transform (Modify) Bone - Bone: auto_skirt_side_lf"!(ComponentPose=@n74, bAlphaBoolEnabled:bool=True, Translation="0,0,0", Rotation="0,0,0", Scale="1,1,1") -> (Pose) {
    alphaInputType=Bool
    rotationMode=BMM_Additive
    rotationSpace=BCS_BoneSpace
    alphaBoolBlend {blendOption=Linear}
    boneToModify {boneName=auto_skirt_side_lf}
   }
   n76 = "Transform (Modify) Bone - Bone: auto_skirt_back_side_lf"!(ComponentPose=@n75, bAlphaBoolEnabled:bool=True, Translation="0,0,0", Rotation="-10,0,0", Scale="1,1,1") -> (Pose) {
    alphaInputType=Bool
    rotationMode=BMM_Additive
    rotationSpace=BCS_BoneSpace
    alphaBoolBlend {blendOption=Linear}
    boneToModify {boneName=auto_skirt_back_side_lf}
    rotation {pitch=-10}
   }
   n77 = "Transform (Modify) Bone - Bone: offset_loin_cloth_front_cr"(ComponentPose=@n76, bAlphaBoolEnabled=True, Translation="0,0,0", Rotation="40,0,0", Scale="1,1,1") {
    alphaInputType=Bool
    rotationMode=BMM_Additive
    rotationSpace=BCS_BoneSpace
    alphaBoolBlend {blendOption=Linear}
    boneToModify {boneName=offset_loin_cloth_front_cr}
    rotation {pitch=40}
   }
   n78 = "Transform (Modify) Bone - Bone: auto_skirt_front_rt"!(ComponentPose=@n77, bAlphaBoolEnabled:bool=True, Translation="0,0,0", Rotation="7.231,-30.404,-4.692", Scale="1,1,1") -> (Pose) {
    alphaInputType=Bool
    rotationMode=BMM_Additive
    rotationSpace=BCS_BoneSpace
    alphaBoolBlend {blendOption=Linear}
    boneToModify {boneName=auto_skirt_front_rt}
    rotation=<rot 7.2311 -30.404 -4.692>
   }
   n79 = "Transform (Modify) Bone - Bone: auto_skirt_front_side_rt"!(ComponentPose=@n78, bAlphaBoolEnabled:bool=True, Translation="0,0,0", Rotation="-24.707,-55.502,-9.484", Scale="1,1,1") -> (Pose) {
    alphaInputType=Bool
    rotationMode=BMM_Additive
    rotationSpace=BCS_BoneSpace
    alphaBoolBlend {blendOption=Linear}
    boneToModify {boneName=auto_skirt_front_side_rt}
    rotation=<rot -24.707 -55.5019 -9.484>
   }
   n80 = "Transform (Modify) Bone - Bone: auto_skirt_side_rt"!(ComponentPose=@n79, bAlphaBoolEnabled:bool=True, Translation="0,0,0", Rotation="-10,0,0", Scale="1,1,1") -> (Pose) {
    alphaInputType=Bool
    rotationMode=BMM_Additive
    rotationSpace=BCS_BoneSpace
    alphaBoolBlend {blendOption=Linear}
    boneToModify {boneName=auto_skirt_side_rt}
    rotation {pitch=-10}
   }
   n81 = "Transform (Modify) Bone - Bone: offset_loin_cloth_front_cr"(ComponentPose=@n80, bAlphaBoolEnabled=@n22, Translation="0,0,0", Rotation="10,20,0", Scale="1,1,1") {
    alphaInputType=Bool
    rotationMode=BMM_Additive
    rotationSpace=BCS_BoneSpace
    alphaBoolBlend {blendOption=Linear}
    boneToModify {boneName=offset_loin_cloth_front_cr}
    rotation {pitch=10, yaw=20}
   }
   n82 = "Transform (Modify) Bone - Bone: offset_skirt_front_side_lf"(ComponentPose=@n81, bAlphaBoolEnabled=@n22, Translation="0,0,0", Rotation="10,20,0", Scale="1,1,1") {
    alphaInputType=Bool
    rotationMode=BMM_Additive
    rotationSpace=BCS_BoneSpace
    alphaBoolBlend {blendOption=Linear}
    boneToModify {boneName=offset_skirt_front_side_lf}
    rotation {pitch=10, yaw=20}
   }
   n83 = "Transform (Modify) Bone - Bone: offset_loin_cloth_back_lf"!(ComponentPose=@n82, bAlphaBoolEnabled:bool=@n22, Translation="0,0,0", Rotation="0,0,0", Scale="1,1,1") -> (Pose) {
    alphaInputType=Bool
    rotationMode=BMM_Additive
    rotationSpace=BCS_BoneSpace
    alphaBoolBlend {blendOption=Linear}
    boneToModify {boneName=offset_loin_cloth_back_lf}
   }
   n84 = "Transform (Modify) Bone - Bone: offset_skirt_back_lf"(ComponentPose=@n83, bAlphaBoolEnabled=@n22, Translation="0,0,0", Rotation="0,0,0", Scale="1,1,1") {
    alphaInputType=Bool
    rotationMode=BMM_Additive
    rotationSpace=BCS_BoneSpace
    alphaBoolBlend {blendOption=Linear}
    boneToModify {boneName=offset_skirt_back_lf}
   }
   n85 = "Transform (Modify) Bone - Bone: offset_loin_cloth_front_cr"(ComponentPose=@n84, bAlphaBoolEnabled=True, Translation="0,0,0", Rotation="10,0,0", Scale="1,1,1") {
    alphaInputType=Bool
    rotationMode=BMM_Additive
    rotationSpace=BCS_BoneSpace
    alphaBoolBlend {blendOption=Linear}
    boneToModify {boneName=offset_loin_cloth_front_cr}
    rotation {pitch=10}
   }
   Save cached pose 'slide_cache'(Pose=Component To Local(ComponentPose=@n85)) {cacheName=slide_cache}
  ]}
 EventGraph {nodes [
   Event BlueprintUpdateAnimation()
   > n1 = Sequence!(execute) -> (then_0, then_1, then_2:Exec, then_3:Exec)
   n2 = Get bIsCrouching()
   Set Is Crouch Moving(execute=@n1.then_1, Is Crouch Moving:bool=AND Boolean(A=Get bIsMoving2D(), B=@n2)) -> (Output_Get:bool)
   Set Is Moving(execute=@n1.then_2, Is Moving:bool=AND Boolean(A=NOT Boolean(A=@n2), B=Get bIsSprinting())) -> (Output_Get:bool)
   n3 = Get bIsPlayingEmote()
   n4 = Get bIsMoving2D()
   n5 = Get bIsCrouching()
   Set Is IDLE(execute=@n1.then_3, Is IDLE:bool=OR Boolean!(A=NOT Boolean(A=@n4), B=NOT Boolean(A=@n5), C:bool=NOT Boolean(A=@n3)) -> (ReturnValue)) -> (Output_Get:bool)
   n6 = Gravity Override State Function DEPRECATED(
    execute=@n1.then_0
    CharacterPartAnimInstance=Self-Reference()
    GameTimeInSeconds=GetGameTimeInSeconds()
    GravityOverrideParams=Get Gravity Override Params()
    RBAN Params=Get RBAN Params()
   )
   > Set GravityOverride(GravityOverride=@n6)
   > Set Sim Space Settings(Sim Space Settings:Sim Space Settings Structure=Break RBANSetting Struct!(RBANSettingStruct:"RBANSetting Struct Structure (by ref)"=@n6.RBAN SettingsOut) -> (
     ComponentLinearAccScale_28_#1:Vector=~
     ComponentLinearVelScale_31_#2:Vector=~
     SimSpaceSettings_34_#3:Sim Space Settings Structure
    ) {is=title K2Node_BreakStruct}.SimSpaceSettings_34_#3) -> (Output_Get:Sim Space Settings Structure)
  ]}
]
variables [
 Gravity Override Params {type=GravityOverrideParamsStruct, default {onGround_Falling {gravityOverride=<0 0 -2000>}, ziplining {gravityOverride=<0 0 0>}}}
 GravityOverride {type=Vector}
 Is Crouch Moving {type=boolean}
 Is IDLE {type=boolean}
 Is Moving {type=boolean}
 RBAN Params {
  type=RBANParamsStruct
  default {
   dBNO {simSpaceSettings {
     maxAngularAcceleration=1000
     maxAngularVelocity=3
     maxLinearAcceleration=300
     maxLinearVelocity=700
     velocityScaleZ=0
     externalLinearDragV=<0 0.5 0>
    }}
   dBNO_Carried {simSpaceSettings {
     maxAngularAcceleration=1000
     maxAngularVelocity=3
     maxLinearAcceleration=300
     maxLinearVelocity=700
     velocityScaleZ=0
     externalLinearDragV=<0 0.5 0>
    }}
   emoteOrMelee {simSpaceSettings {
     maxAngularAcceleration=1000
     maxAngularVelocity=3
     maxLinearAcceleration=300
     maxLinearVelocity=700
     velocityScaleZ=0
     externalLinearDragV=<0 0.5 0>
    }}
   front_End {simSpaceSettings {
     maxAngularAcceleration=1000
     maxAngularVelocity=3
     maxLinearAcceleration=300
     maxLinearVelocity=700
     velocityScaleZ=0
     externalLinearDragV=<0 0.5 0>
    }}
   onGround_Crouch_Moving {simSpaceSettings {
     maxAngularAcceleration=1000
     maxAngularVelocity=3
     maxLinearAcceleration=300
     maxLinearVelocity=700
     velocityScaleZ=0
     externalLinearDragV=<0 0.5 0>
    }}
   onGround_Falling {simSpaceSettings {
     maxAngularAcceleration=1000
     maxAngularVelocity=3
     maxLinearAcceleration=300
     maxLinearVelocity=700
     velocityScaleZ=1
     externalLinearDragV=<0 0.5 0>
    }}
   onGround_Moving {
    componentLinearAccScale=<0.3 1 0>
    componentLinearVelScale=<0.3 1 0>
    simSpaceSettings {
     maxAngularAcceleration=1000
     maxAngularVelocity=3
     maxLinearAcceleration=300
     maxLinearVelocity=700
     velocityScaleZ=0
     externalLinearDragV=<0 0.5 0>
    }
   }
   onGround_Standing {
    componentLinearAccScale=<0.3 0.5 0>
    componentLinearVelScale=<0.3 0.5 0>
    simSpaceSettings {
     maxAngularAcceleration=1000
     maxAngularVelocity=1000
     maxLinearAcceleration=10000
     maxLinearVelocity=10000
     velocityScaleZ=0
     externalLinearDragV=<0 0.5 0>
     externalLinearVelocity=<0 1 0>
    }
   }
   rail_Grinding {simSpaceSettings {
     maxAngularAcceleration=1000
     maxAngularVelocity=3
     maxLinearAcceleration=300
     maxLinearVelocity=700
     velocityScaleZ=0
     worldAlpha=0.1
     externalLinearDragV=<0 0.5 0>
    }}
   riding_Moving {simSpaceSettings {
     maxAngularAcceleration=1000
     maxAngularVelocity=3
     maxLinearAcceleration=300
     maxLinearVelocity=700
     velocityScaleZ=0
     externalLinearDragV=<0 0.5 0>
    }}
   riding_Moving_Targeting {simSpaceSettings {
     maxAngularAcceleration=1000
     maxAngularVelocity=3
     maxLinearAcceleration=300
     maxLinearVelocity=700
     velocityScaleZ=0
     externalLinearDragV=<0 0.5 0>
    }}
   riding_Standing {simSpaceSettings {
     maxAngularAcceleration=1000
     maxAngularVelocity=3
     maxLinearAcceleration=300
     maxLinearVelocity=700
     velocityScaleZ=0
     externalLinearDragV=<0 0.5 0>
    }}
   skydive_DivingDown {simSpaceSettings {
     maxAngularAcceleration=1000
     maxAngularVelocity=3
     maxLinearAcceleration=300
     maxLinearVelocity=700
     velocityScaleZ=0
     externalLinearDragV=<0 0.5 0>
    }}
   skydive_FloatingDown {
    componentLinearAccScale=<0 0.5 0>
    componentLinearVelScale=<0 0.5 0>
    simSpaceSettings {
     maxAngularAcceleration=1000
     maxAngularVelocity=3
     maxLinearAcceleration=300
     maxLinearVelocity=700
     velocityScaleZ=0
     externalLinearDragV=<0 0.5 0>
    }
   }
   skydive_GoingUp {simSpaceSettings {
     maxAngularAcceleration=1000
     maxAngularVelocity=3
     maxLinearAcceleration=300
     maxLinearVelocity=700
     velocityScaleZ=0
     externalLinearDragV=<0 0.5 0>
    }}
   skydive_Parachuting {
    componentLinearAccScale=<0 0.1 0.3>
    componentLinearVelScale=<0 0.1 0.3>
    simSpaceSettings {
     maxAngularAcceleration=1000
     maxAngularVelocity=3
     maxLinearAcceleration=300
     maxLinearVelocity=700
     velocityScaleZ=0
     externalLinearDragV=<0 0.5 0>
    }
   }
   skydive_Surfing {simSpaceSettings {
     maxAngularAcceleration=1000
     maxAngularVelocity=3
     maxLinearAcceleration=300
     maxLinearVelocity=700
     velocityScaleZ=0
     externalLinearDragV=<0 0.5 0>
    }}
   swimming {simSpaceSettings {
     maxAngularAcceleration=1000
     maxAngularVelocity=3
     maxLinearAcceleration=300
     maxLinearVelocity=700
     velocityScaleZ=0
     externalLinearDragV=<0 0.5 0>
    }}
   tactical_Sprint {simSpaceSettings {
     maxAngularAcceleration=1000
     maxAngularVelocity=3
     maxLinearAcceleration=300
     maxLinearVelocity=700
     velocityScaleZ=0
     externalLinearDragV=<0 0.5 0>
    }}
   ziplining {simSpaceSettings {
     maxAngularAcceleration=1000
     maxAngularVelocity=3
     maxLinearAcceleration=300
     maxLinearVelocity=700
     velocityScaleZ=0
     worldAlpha=0
     externalLinearDragV=<0 0.5 0>
    }}
  }
 }
 Sim Space Settings {
  type=SimSpaceSettings
  default {
   dampingAlpha=1
   maxAngularAcceleration=10000
   maxAngularVelocity=10000
   maxLinearAcceleration=10000
   maxLinearVelocity=10000
   velocityScaleZ=1
   worldAlpha=0
   externalLinearDragV=<0 0 0>
  }
 }
]
